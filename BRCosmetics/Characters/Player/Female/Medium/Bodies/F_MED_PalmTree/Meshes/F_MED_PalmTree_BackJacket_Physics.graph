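# Physics asset for the back jacket of the female medium PalmTree cosmetic body, setting up its dangling sleeve bones for simulated secondary motion.
# Each sleeve is a capsule chain constrained link to link from a small base capsule, four dynamic bones on the left and three on the right, with narrow swing limits and zero twist limit.
# Unconstrained capsules on both thighs appear to serve as leg colliders for the sleeves.
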PhysicsAsset /BRCosmetics/Characters/Player/Female/Medium/Bodies/F_MED_PalmTree/Meshes/F_MED_PalmTree_BackJacket_Physics
bodies {
 base_sleeve_l [{shapeType=Capsule, length=0.505, radius=0.505, shapeName=base_sleeve_l_capsule, rotation=<rot 0 -90 0>}]
 base_sleeve_r [{shapeType=Capsule, length=0.505, radius=0.505, shapeName=base_sleeve_r_capsule, rotation=<rot 0 -90 0>}]
 dyn_sleeve_l_1 [{shapeType=Capsule, length=2.0831, radius=1.4917, shapeName=dyn_sleeve_l_1_capsule, center=<2.9037 0.093 0.7101>, rotation=<rot 90 180 0>}]
 dyn_sleeve_l_2 [{shapeType=Capsule, length=5.1161, radius=2.6429, shapeName=dyn_sleeve_l_2_capsule, center=<2.892 -0.1312 1.4735>, rotation=<rot 90 0 0>}]
 dyn_sleeve_l_3 [{shapeType=Capsule, length=4.608, radius=2.8559, shapeName=dyn_sleeve_l_3_capsule, center=<4.3283 -0.1312 1.7311>, rotation=<rot 90 0 0>}]
 dyn_sleeve_l_4 [
  {shapeType=Capsule, length=4.3934, radius=2.814, shapeName=dyn_sleeve_l_4_capsule, center=<1.0598 -0.1925 0.8633>, rotation=<rot 0 0 90>}
  {shapeType=Capsule, length=4.3934, radius=2.814, shapeName=dyn_sleeve_l_4_capsule, center=<4.3273 -0.1925 0.4654>, rotation=<rot 0 0 90>}
 ]
 dyn_sleeve_r_1 [{shapeType=Capsule, length=4.4625, radius=2.4961, shapeName=dyn_sleeve_r_1_capsule, center=<3.5592 -0.399 1.1128>, rotation=<rot 0 0 90>}]
 dyn_sleeve_r_2 [{shapeType=Capsule, length=3.8141, radius=2.6746, shapeName=dyn_sleeve_r_2_capsule, center=<3.8697 -0.399 0.9763>, rotation=<rot 0 0 90>}]
 dyn_sleeve_r_3 [{shapeType=Capsule, length=3.3694, radius=3.2644, shapeName=dyn_sleeve_r_3_capsule, center=<4.3683 0 0.8034>, rotation=<rot 0 0 90>}]
 thigh_l [{
   shapeType=Capsule
   length=32.3544
   radius=9.3214
   shapeName=thigh_l_capsule
   center=<-19.4883 -2.2012 -0.2153>
   rotation=<rot 0.0256 -89.1469 -90.5273>
  }]
 thigh_r [{
   shapeType=Capsule
   length=32.3544
   radius=9.3214
   shapeName=thigh_l_capsule
   center=<19.4864 1.7518 0.2148>
   rotation=<rot 0.0203 -89.7237 -90.5276>
  }]
}
constraints [
 {bone1Name=dyn_sleeve_l_1, bone2Name=base_sleeve_l, swing1LimitDegrees=15, swing2LimitDegrees=15, twistLimitDegrees=0}
 {bone1Name=dyn_sleeve_l_2, bone2Name=dyn_sleeve_l_1, swing1LimitDegrees=15, swing2LimitDegrees=15, twistLimitDegrees=0}
 {bone1Name=dyn_sleeve_l_3, bone2Name=dyn_sleeve_l_2, swing1LimitDegrees=15, swing2LimitDegrees=15, twistLimitDegrees=0}
 {bone1Name=dyn_sleeve_l_4, bone2Name=dyn_sleeve_l_3, swing1LimitDegrees=15, swing2LimitDegrees=15, twistLimitDegrees=0}
 {bone1Name=dyn_sleeve_r_1, bone2Name=base_sleeve_r, swing1LimitDegrees=15, swing2LimitDegrees=15, twistLimitDegrees=0}
 {bone1Name=dyn_sleeve_r_2, bone2Name=dyn_sleeve_r_1, swing1LimitDegrees=15, swing2LimitDegrees=15, twistLimitDegrees=0}
 {bone1Name=dyn_sleeve_r_3, bone2Name=dyn_sleeve_r_2, swing1LimitDegrees=15, swing2LimitDegrees=15, twistLimitDegrees=0}
]
