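# Physics asset for the F_MED_PalmTree female player cosmetic body, set up for secondary motion on its skirt, leg pouches and straps.
# Seven skirt capsule chains (A–G) hang from root-linked base bodies, with tight swing limits and mostly zero twist.
# Each thigh-linked pouch carries front and back three-link strap chains with limited swing.
# Body capsules cover only root, both thighs, and the right calf and foot.
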PhysicsAsset /BRCosmetics/Characters/Player/Female/Medium/Bodies/F_MED_PalmTree/Meshes/F_MED_PalmTree_Physics
bodies {
 base_skirt_A [{shapeType=Capsule, length=0.505, radius=0.505, shapeName=base_skirt_A_capsule, rotation=<rot 0 -90 0>}]
 base_skirt_B [{shapeType=Capsule, length=0.505, radius=0.505, shapeName=base_skirt_B_capsule, rotation=<rot 0 -90 0>}]
 base_skirt_C [{shapeType=Capsule, length=0.505, radius=0.505, shapeName=base_skirt_C_capsule, rotation=<rot 0 -90 0>}]
 base_skirt_D [{shapeType=Capsule, length=0.505, radius=0.505, shapeName=base_skirt_D_capsule, rotation=<rot 0 -90 0>}]
 base_skirt_E [{shapeType=Capsule, length=0.505, radius=0.505, shapeName=base_skirt_E_capsule, rotation=<rot 0 -90 0>}]
 base_skirt_F [{shapeType=Capsule, length=0.505, radius=0.505, shapeName=base_skirt_F_capsule, rotation=<rot 0 -90 0>}]
 base_skirt_G [{shapeType=Capsule, length=0.505, radius=0.505, shapeName=base_skirt_G_capsule, rotation=<rot 0 -90 0>}]
 base_strapBack_l [{shapeType=Capsule, length=0.505, radius=0.505, shapeName=base_strapBack_l_capsule, rotation=<rot 0 -90 0>}]
 base_strapBack_r [{shapeType=Capsule, length=0.505, radius=0.505, shapeName=base_strapBack_r_capsule, rotation=<rot 0 -90 0>}]
 base_strapFront_l [{shapeType=Capsule, length=0.505, radius=0.505, shapeName=base_strapFront_l_capsule, rotation=<rot 0 -90 0>}]
 base_strapFront_r [{shapeType=Capsule, length=0.505, radius=0.505, shapeName=base_strapFront_r_capsule, rotation=<rot 0 -90 0>}]
 calf_r [{
   shapeType=Capsule
   length=29.4725
   radius=9.6894
   shapeName=calf_l_capsule
   center=<22.3455 1.5656 0.3163>
   rotation=<rot 0.5618 89.5579 89.7343>
  }]
 dyn_legPouch_l [{shapeType=Capsule, length=6.6819, radius=7.6004, shapeName=dyn_legPouch_l_capsule, center=<0 0 2.5>, rotation=<rot 90 0 0>}]
 dyn_legPouch_r [{shapeType=Capsule, length=6.6819, radius=7.6004, shapeName=dyn_legPouch_l_capsule, center=<0 0 -2.5>, rotation=<rot 90 0 0>}]
 dyn_skirt_A_02 [{shapeType=Capsule, length=2.5, radius=2, shapeName=dyn_skirt_A_02_capsule, center=<2.4439 -0.4825 0.8648>, rotation=<rot 90 0 0>}]
 dyn_skirt_A_03 [{shapeType=Capsule, length=4.0955, radius=2.9045, shapeName=dyn_skirt_A_03_capsule, center=<3.2956 -1.9251 0>, rotation=<rot 90 0 0>}]
 dyn_skirt_A_04 [{shapeType=Capsule, length=3.0328, radius=3.5496, shapeName=dyn_skirt_A_04_capsule, center=<3.4695 -2.2921 0>, rotation=<rot 90 0 0>}]
 dyn_skirt_B_02 [{shapeType=Capsule, length=6.5, radius=2.5, shapeName=dyn_skirt_B_02_capsule, center=<5.9358 0 0>}]
 dyn_skirt_B_03 [{shapeType=Capsule, length=11, radius=4, shapeName=dyn_skirt_B_03_capsule, center=<4.6301 0 0>, rotation=<rot 0 0 -5.625>}]
 dyn_skirt_B_04 [{shapeType=Capsule, length=11, radius=4, shapeName=dyn_skirt_B_04_capsule, center=<3.6779 -1.343 -0.3364>, rotation=<rot 0.0001 0 -14.0625>}]
 dyn_skirt_C_01 [{shapeType=Capsule, length=5, radius=3, shapeName=dyn_skirt_C_01_capsule, center=<6.4853 0 0>}]
 dyn_skirt_C_02 [{shapeType=Capsule, length=6.5, radius=3, shapeName=dyn_skirt_C_02_capsule, center=<5.0812 0 0>}]
 dyn_skirt_C_03 [{shapeType=Capsule, length=7.5, radius=4, shapeName=dyn_skirt_C_03_capsule, center=<6.2861 0 0>}]
 dyn_skirt_C_04 [{shapeType=Capsule, length=8.5, radius=4.5, shapeName=dyn_skirt_C_04_capsule, center=<4.716 0 0>}]
 dyn_skirt_D_01 [{shapeType=Capsule, length=5, radius=3, shapeName=dyn_skirt_D_01_capsule, center=<6.4853 0 0>}]
 dyn_skirt_D_02 [{shapeType=Capsule, length=7, radius=3, shapeName=dyn_skirt_D_02_capsule, center=<4.6817 0 0>}]
 dyn_skirt_D_03 [{shapeType=Capsule, length=6, radius=4, shapeName=dyn_skirt_D_03_capsule, center=<4.4009 0 0>}]
 dyn_skirt_D_04 [{shapeType=Capsule, length=7, radius=4.5, shapeName=dyn_skirt_D_04_capsule, center=<2.1419 0 0>}]
 dyn_skirt_E_01 [{shapeType=Capsule, length=5, radius=3, shapeName=dyn_skirt_E_01_capsule, center=<6.4853 0 0>}]
 dyn_skirt_E_02 [{shapeType=Capsule, length=6.5, radius=3, shapeName=dyn_skirt_E_02_capsule, center=<5.6279 -1.1213 0>}]
 dyn_skirt_E_03 [{shapeType=Capsule, length=7.5, radius=4, shapeName=dyn_skirt_E_03_capsule, center=<6.0463 0 0>}]
 dyn_skirt_E_04 [{shapeType=Capsule, length=5, radius=4.5, shapeName=dyn_skirt_E_04_capsule, center=<6.0548 0 0>}]
 dyn_skirt_F_02 [{shapeType=Capsule, length=8, radius=2.5, shapeName=dyn_skirt_F_02_capsule, center=<4.8927 0.1653 0>}]
 dyn_skirt_F_03 [{shapeType=Capsule, length=9, radius=4, shapeName=dyn_skirt_F_03_capsule, center=<3.6099 -1.051 0>}]
 dyn_skirt_F_04 [{shapeType=Capsule, length=9, radius=4.5, shapeName=dyn_skirt_F_04_capsule, center=<2.66 -0.7646 0>}]
 dyn_skirt_G_02 [{shapeType=Capsule, length=3.2793, radius=3.5, shapeName=dyn_skirt_G_02_capsule, center=<2.4247 -1.7907 1.4439>, rotation=<rot 90 0 0>}]
 dyn_skirt_G_03 [{shapeType=Capsule, length=4.0941, radius=2.9059, shapeName=dyn_skirt_G_03_capsule, center=<4.8152 -1.5527 0>, rotation=<rot 90 0 0>}]
 dyn_skirt_G_04 [{shapeType=Capsule, length=2.5621, radius=3.9379, shapeName=dyn_skirt_G_04_capsule, center=<4.1154 -1.5527 0>, rotation=<rot 90 0 0>}]
 dyn_strapBack_1_l [{shapeType=Capsule, length=1, radius=1, shapeName=dyn_strapBack_1_l_capsule, center=<-2.9702 0 0>, rotation=<rot 90 0 0>}]
 dyn_strapBack_1_r [{shapeType=Capsule, length=1, radius=1, shapeName=dyn_strapBack_1_r_capsule, center=<2.9754 0 0>, rotation=<rot 90 0 0>}]
 dyn_strapBack_2_l [{shapeType=Capsule, length=1, radius=1, shapeName=dyn_strapBack_2_l_capsule, center=<-1.8481 0 0>, rotation=<rot 90 0 0>}]
 dyn_strapBack_2_r [{shapeType=Capsule, length=1, radius=1, shapeName=dyn_strapBack_2_r_capsule, center=<1.8492 0 0>, rotation=<rot 90 0 0>}]
 dyn_strapBack_3_l [{shapeType=Capsule, length=1, radius=1, shapeName=dyn_strapBack_3_l_capsule, center=<-1.5356 0 0>, rotation=<rot 90 0 0>}]
 dyn_strapBack_3_r [{shapeType=Capsule, length=1, radius=1, shapeName=dyn_strapBack_3_r_capsule, center=<1.5341 0 0>, rotation=<rot 90 0 0>}]
 dyn_strapFront_1_l [{shapeType=Capsule, length=1, radius=1, shapeName=dyn_strapFront_1_l_capsule, center=<-2.9169 0 0>, rotation=<rot 90 0 0>}]
 dyn_strapFront_1_r [{shapeType=Capsule, length=1, radius=1, shapeName=dyn_strapFront_1_r_capsule, center=<2.9436 0 0>, rotation=<rot 90 0 0>}]
 dyn_strapFront_2_l [{shapeType=Capsule, length=1, radius=1, shapeName=dyn_strapFront_2_l_capsule, center=<-1.8876 0 0>, rotation=<rot 90 0 0>}]
 dyn_strapFront_2_r [{shapeType=Capsule, length=1, radius=1, shapeName=dyn_strapFront_2_r_capsule, center=<1.8876 0 0>, rotation=<rot 90 0 0>}]
 dyn_strapFront_3_l [{shapeType=Capsule, length=1, radius=1, shapeName=dyn_strapFront_3_l_capsule, center=<-1.5757 0 0>, rotation=<rot 90 0 0>}]
 dyn_strapFront_3_r [{shapeType=Capsule, length=1, radius=1, shapeName=dyn_strapFront_3_r_capsule, center=<1.5757 0 0>, rotation=<rot 90 0 0>}]
 foot_r [{shapeType=Capsule, length=15.1639, radius=7.7021, shapeName=foot_l_capsule, center=<3.8396 -7.0072 0.2697>, rotation=<rot 0 -3.2513 91.4339>}]
 root [{shapeType=Capsule, length=20.644, radius=22.6135, shapeName=root_capsule, center=<0 -6.8959 18.1072>, rotation=<rot 0 0 30.9375>}]
 thigh_l [{
   shapeType=Capsule
   length=29.6392
   radius=8.5481
   shapeName=thigh_l_capsule
   center=<-20.6847 -0.8248 1.2915>
   rotation=<rot 0 -91.9195 -90.5279>
  }]
 thigh_r [{
   shapeType=Capsule
   length=28.2642
   radius=9.11
   shapeName=thigh_l_capsule
   center=<20.1387 0.8431 -0.7982>
   rotation=<rot 0 -91.9195 89.4721>
  }]
}
constraints [
 {
  bone1Name=thigh_l
  bone2Name=root
  swing1LimitDegrees=45
  swing1Motion=Free
  swing2LimitDegrees=45
  swing2Motion=Free
  twistLimitDegrees=45
  twistMotion=Free
 }
 {
  bone1Name=thigh_r
  bone2Name=root
  swing1LimitDegrees=45
  swing1Motion=Free
  swing2LimitDegrees=45
  swing2Motion=Free
  twistLimitDegrees=45
  twistMotion=Free
 }
 {
  bone1Name=calf_r
  bone2Name=thigh_r
  swing1LimitDegrees=45
  swing1Motion=Free
  swing2LimitDegrees=45
  swing2Motion=Free
  twistLimitDegrees=45
  twistMotion=Free
 }
 {
  bone1Name=foot_r
  bone2Name=calf_r
  swing1LimitDegrees=45
  swing1Motion=Free
  swing2LimitDegrees=45
  swing2Motion=Free
  twistLimitDegrees=45
  twistMotion=Free
 }
 {
  bone1Name=base_skirt_G
  bone2Name=root
  swing1LimitDegrees=45
  swing1Motion=Free
  swing2LimitDegrees=45
  swing2Motion=Free
  twistLimitDegrees=45
  twistMotion=Free
 }
 {
  bone1Name=base_skirt_F
  bone2Name=root
  swing1LimitDegrees=45
  swing1Motion=Free
  swing2LimitDegrees=45
  swing2Motion=Free
  twistLimitDegrees=45
  twistMotion=Free
 }
 {
  bone1Name=base_skirt_B
  bone2Name=root
  swing1LimitDegrees=45
  swing1Motion=Free
  swing2LimitDegrees=45
  swing2Motion=Free
  twistLimitDegrees=45
  twistMotion=Free
 }
 {bone1Name=dyn_skirt_A_02, bone2Name=base_skirt_A, swing1LimitDegrees=2, swing2LimitDegrees=10, twistLimitDegrees=0}
 {bone1Name=dyn_skirt_A_03, bone2Name=dyn_skirt_A_02, swing1LimitDegrees=5, swing2LimitDegrees=12, twistLimitDegrees=0}
 {bone1Name=dyn_skirt_B_02, bone2Name=base_skirt_B, swing1LimitDegrees=2, swing2LimitDegrees=5, twistLimitDegrees=5}
 {bone1Name=dyn_skirt_B_03, bone2Name=dyn_skirt_B_02, swing1LimitDegrees=5, swing2LimitDegrees=12, twistLimitDegrees=0}
 {bone1Name=dyn_skirt_B_04, bone2Name=dyn_skirt_B_03, swing1LimitDegrees=5, swing2LimitDegrees=12, twistLimitDegrees=0}
 {bone1Name=dyn_skirt_F_02, bone2Name=base_skirt_F, swing1LimitDegrees=2, swing2LimitDegrees=5, twistLimitDegrees=0}
 {bone1Name=dyn_skirt_F_03, bone2Name=dyn_skirt_F_02, swing1LimitDegrees=5, swing2LimitDegrees=10, twistLimitDegrees=0}
 {bone1Name=dyn_skirt_F_04, bone2Name=dyn_skirt_F_03, swing1LimitDegrees=5, swing2LimitDegrees=10, twistLimitDegrees=0}
 {bone1Name=dyn_skirt_G_02, bone2Name=base_skirt_G, swing1LimitDegrees=2, swing2LimitDegrees=10, twistLimitDegrees=0}
 {bone1Name=dyn_skirt_G_03, bone2Name=dyn_skirt_G_02, swing1LimitDegrees=5, swing2LimitDegrees=12, twistLimitDegrees=0}
 {bone1Name=dyn_skirt_G_04, bone2Name=dyn_skirt_G_03, swing1LimitDegrees=5, swing2LimitDegrees=12, twistLimitDegrees=0}
 {bone1Name=dyn_skirt_A_04, bone2Name=dyn_skirt_A_03, swing1LimitDegrees=5, swing2LimitDegrees=12, twistLimitDegrees=0}
 {
  bone1Name=base_skirt_A
  bone2Name=root
  swing1LimitDegrees=45
  swing1Motion=Free
  swing2LimitDegrees=45
  swing2Motion=Free
  twistLimitDegrees=45
  twistMotion=Free
 }
 {
  bone1Name=base_skirt_C
  bone2Name=root
  swing1LimitDegrees=45
  swing1Motion=Free
  swing2LimitDegrees=45
  swing2Motion=Free
  twistLimitDegrees=45
  twistMotion=Free
 }
 {
  bone1Name=base_skirt_D
  bone2Name=root
  swing1LimitDegrees=45
  swing1Motion=Free
  swing2LimitDegrees=45
  swing2Motion=Free
  twistLimitDegrees=45
  twistMotion=Free
 }
 {
  bone1Name=base_skirt_E
  bone2Name=root
  swing1LimitDegrees=45
  swing1Motion=Free
  swing2LimitDegrees=45
  swing2Motion=Free
  twistLimitDegrees=45
  twistMotion=Free
 }
 {bone1Name=dyn_skirt_C_01, bone2Name=base_skirt_C, swing1LimitDegrees=10, swing2LimitDegrees=2, twistLimitDegrees=0}
 {bone1Name=dyn_skirt_C_02, bone2Name=dyn_skirt_C_01, swing1LimitDegrees=10, swing2LimitDegrees=2, twistLimitDegrees=0}
 {bone1Name=dyn_skirt_C_03, bone2Name=dyn_skirt_C_02, swing1LimitDegrees=12, swing2LimitDegrees=5, twistLimitDegrees=0}
 {bone1Name=dyn_skirt_C_04, bone2Name=dyn_skirt_C_03, swing1LimitDegrees=12, swing2LimitDegrees=5, twistLimitDegrees=0}
 {bone1Name=dyn_skirt_D_01, bone2Name=base_skirt_D, swing1LimitDegrees=10, swing2LimitDegrees=0, twistLimitDegrees=0}
 {bone1Name=dyn_skirt_D_02, bone2Name=dyn_skirt_D_01, swing1LimitDegrees=10, swing2LimitDegrees=0, twistLimitDegrees=0}
 {bone1Name=dyn_skirt_D_03, bone2Name=dyn_skirt_D_02, swing1LimitDegrees=12, swing2LimitDegrees=0, twistLimitDegrees=0}
 {bone1Name=dyn_skirt_D_04, bone2Name=dyn_skirt_D_03, swing1LimitDegrees=12, swing2LimitDegrees=0, twistLimitDegrees=0}
 {bone1Name=dyn_skirt_E_01, bone2Name=base_skirt_E, swing1LimitDegrees=10, swing2LimitDegrees=2, twistLimitDegrees=0}
 {bone1Name=dyn_skirt_E_02, bone2Name=dyn_skirt_E_01, swing1LimitDegrees=10, swing2LimitDegrees=2, twistLimitDegrees=0}
 {bone1Name=dyn_skirt_E_03, bone2Name=dyn_skirt_E_02, swing1LimitDegrees=12, swing2LimitDegrees=5, twistLimitDegrees=0}
 {bone1Name=dyn_skirt_E_04, bone2Name=dyn_skirt_E_03, swing1LimitDegrees=12, swing2LimitDegrees=5, twistLimitDegrees=0}
 {
  bone1Name=base_strapBack_r
  bone2Name=dyn_legPouch_r
  swing1LimitDegrees=45
  swing1Motion=Free
  swing2LimitDegrees=45
  swing2Motion=Free
  twistLimitDegrees=45
  twistMotion=Free
 }
 {
  bone1Name=base_strapFront_r
  bone2Name=dyn_legPouch_r
  swing1LimitDegrees=45
  swing1Motion=Free
  swing2LimitDegrees=45
  swing2Motion=Free
  twistLimitDegrees=45
  twistMotion=Free
 }
 {bone1Name=dyn_strapBack_1_r, bone2Name=base_strapBack_r, swing1LimitDegrees=15, swing2LimitDegrees=5, twistLimitDegrees=0}
 {bone1Name=dyn_strapBack_2_r, bone2Name=dyn_strapBack_1_r, swing1LimitDegrees=15, swing2LimitDegrees=5, twistLimitDegrees=0}
 {bone1Name=dyn_strapBack_3_r, bone2Name=dyn_strapBack_2_r, swing1LimitDegrees=20, swing2LimitDegrees=10, twistLimitDegrees=0}
 {bone1Name=dyn_strapFront_1_r, bone2Name=base_strapFront_r, swing1LimitDegrees=15, swing2LimitDegrees=5, twistLimitDegrees=0}
 {bone1Name=dyn_strapFront_2_r, bone2Name=dyn_strapFront_1_r, swing1LimitDegrees=15, swing2LimitDegrees=5, twistLimitDegrees=0}
 {bone1Name=dyn_strapFront_3_r, bone2Name=dyn_strapFront_2_r, swing1LimitDegrees=20, swing2LimitDegrees=10, twistLimitDegrees=0}
 {
  bone1Name=base_strapFront_l
  bone2Name=dyn_legPouch_l
  swing1LimitDegrees=45
  swing1Motion=Free
  swing2LimitDegrees=45
  swing2Motion=Free
  twistLimitDegrees=45
  twistMotion=Free
 }
 {
  bone1Name=base_strapBack_l
  bone2Name=dyn_legPouch_l
  swing1LimitDegrees=45
  swing1Motion=Free
  swing2LimitDegrees=45
  swing2Motion=Free
  twistLimitDegrees=45
  twistMotion=Free
 }
 {bone1Name=dyn_strapBack_1_l, bone2Name=base_strapBack_l, swing1LimitDegrees=15, swing2LimitDegrees=5, twistLimitDegrees=0}
 {bone1Name=dyn_strapBack_2_l, bone2Name=dyn_strapBack_1_l, swing1LimitDegrees=15, swing2LimitDegrees=5, twistLimitDegrees=0}
 {bone1Name=dyn_strapBack_3_l, bone2Name=dyn_strapBack_2_l, swing1LimitDegrees=20, swing2LimitDegrees=10, twistLimitDegrees=0}
 {bone1Name=dyn_strapFront_1_l, bone2Name=base_strapFront_l, swing1LimitDegrees=15, swing2LimitDegrees=5, twistLimitDegrees=0}
 {bone1Name=dyn_strapFront_2_l, bone2Name=dyn_strapFront_1_l, swing1LimitDegrees=15, swing2LimitDegrees=5, twistLimitDegrees=0}
 {bone1Name=dyn_strapFront_3_l, bone2Name=dyn_strapFront_2_l, swing1LimitDegrees=20, swing2LimitDegrees=10, twistLimitDegrees=0}
 {
  bone1Name=dyn_legPouch_l
  bone2Name=thigh_l
  swing1LimitDegrees=45
  swing1Motion=Free
  swing2LimitDegrees=45
  swing2Motion=Free
  twistLimitDegrees=45
  twistMotion=Free
 }
 {
  bone1Name=dyn_legPouch_r
  bone2Name=thigh_r
  swing1LimitDegrees=45
  swing1Motion=Free
  swing2LimitDegrees=45
  swing2Motion=Free
  twistLimitDegrees=45
  twistMotion=Free
 }
]
