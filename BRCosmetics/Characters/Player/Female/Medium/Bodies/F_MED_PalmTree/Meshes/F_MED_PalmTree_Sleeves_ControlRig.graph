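# Control Rig for the sleeves mesh of the F_MED_PalmTree female body, posing its waist skirt bones around leg motion and movement states.
# Per side, a parent constraint ties the outer skirt bone to the groin deform bone, and rotation-and-scale constraints blend neighboring bones.
# Spherical pose readers on each thigh add to Crouching to weight corrective offsets.
# FrontEndAlpha, SkydiveMode and TacticalJumping weight further offsets through unclamped remaps.
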ControlRigBlueprint /BRCosmetics/Characters/Player/Female/Medium/Bodies/F_MED_PalmTree/Meshes/F_MED_PalmTree_Sleeves_ControlRig
graphs [
 RigVMFunctionLibrary {links [], nodes []}
 RigVMModel {nodes [
   RigUnit_BeginExecution = &RigUnit_BeginExecution()
   SequenceAggregate_3_1_1 = RigVMAggregateNode(ExecuteContext=@RigUnit_BeginExecution.ExecutePin)
   &RigUnit_ModifyTransforms(
    ItemToModify="((Item=(Type=Bone,Name=\"base_skirt_A\"),Transform=(Rotation=(X=0,Y=0,Z=-0.1305,W=0.9914),Translation=(X=0,Y=0,Z=0),Scale3D=(X=1,Y=1,Z=1))),(Item=(Type=Bone,Name=\"base_skirt_B\"),Transform=(Rotation=(X=0,Y=0,Z=-0.1305,W=0.9914),Translation=(X=0,Y=0,Z=0),Scale ...(truncated)"
    WeightMinimum=0
    Weight=&RigVMFunction_MathDoubleRemap(SourceMaximum=1.2, SourceMinimum=0, TargetMaximum=1, TargetMinimum=0, bClamp=False, Value=RigVMVariableNode() {variable=SkydiveMode}.Value).Result
    ExecutePin=@SequenceAggregate_3_1_1.E
   )
   &RigUnit_ModifyTransforms(
    ItemToModify="((Item=(Type=Bone,Name=\"base_skirt_B\"),Transform=(Rotation=(X=0.0189,Y=0.0851,Z=-0.2156,W=0.9726),Translation=(X=0,Y=-3.5,Z=3.5),Scale3D=(X=1,Y=1,Z=1))),(Item=(Type=Bone,Name=\"base_skirt_C\"),Transform=(Rotation=(X=0,Y=0,Z=-0.1305,W=0.9914),Translation=(X=0 ...(truncated)"
    WeightMinimum=0
    Weight=&RigVMFunction_MathDoubleRemap(
     SourceMaximum=1.2
     SourceMinimum=0
     TargetMaximum=1
     TargetMinimum=0
     bClamp=False
     Value=RigVMVariableNode() {variable=TacticalJumping}.Value
    ).Result
    ExecutePin=@SequenceAggregate_3_1_1.F
   )
   &RigUnit_ParentConstraint(
    AdvancedSettings=(InterpolationType=Average,RotationOrderForFilter=XZY)
    Child=(Type=Bone,Name="base_skirt_A")
    Parents=((Item=(Type=Bone,Name="deform_groin_l"),Weight=1))
    Weight=1
    ExecutePin=@SequenceAggregate_3_1_1.A
   )
   > &RigUnit_ParentConstraint(
    AdvancedSettings=(RotationOrderForFilter=XZY)
    Child=(Type=Bone,Name="base_skirt_B")
    Filter=(TranslationFilter=(bX=False,bY=False,bZ=False),RotationFilter=(bX=True,bY=True,bZ=True),ScaleFilter=(bX=True,bY=True,bZ=True))
    Parents=((Item=(Type=Bone,Name="dyn_skirt_A_01"),Weight=0.5),(Item=(Type=Bone,Name="base_skirt_c"),Weight=0.5))
    Weight=1
   )
   > &RigUnit_ParentConstraint(
    AdvancedSettings=(RotationOrderForFilter=XZY)
    Child=(Type=Bone,Name="base_skirt_c")
    Filter=(TranslationFilter=(bX=False,bY=False,bZ=False),RotationFilter=(bX=True,bY=True,bZ=True),ScaleFilter=(bX=True,bY=True,bZ=True))
    Parents=((Item=(Type=Bone,Name="base_skirt_B"),Weight=0.5),(Item=(Type=Bone,Name="base_skirt_D"),Weight=0.5))
    Weight=1
   )
   n1 = &RigVMFunction_MathDoubleRemap(SourceMaximum=1.2, SourceMinimum=0, TargetMaximum=1, TargetMinimum=0, bClamp=False, Value=RigVMVariableNode() {variable=FrontEndAlpha}.Value)
   > &RigUnit_ModifyTransforms(
    ItemToModify="((Item=(Type=Bone,Name=\"dyn_skirt_A_01\"),Transform=(Rotation=(X=0,Y=0,Z=-0.0436,W=0.999),Translation=(X=0,Y=0,Z=0),Scale3D=(X=1,Y=1,Z=1))),(Item=(Type=Bone,Name=\"dyn_skirt_B_01\"),Transform=(Rotation=(X=0,Y=0,Z=-0.0436,W=0.999),Translation=(X=0,Y=0,Z=0),Sca ...(truncated)"
    WeightMinimum=0
    Weight=@n1.Result
   )
   > n2 = &RigUnit_SphericalPoseReader(
    ActiveRegionScaleFactors=(PositiveWidth=3.5,NegativeWidth=1,PositiveHeight=3,NegativeHeight=3)
    ActiveRegionSize=0.1
    Debug=(bDrawDebug=false,bDraw2D=false,DebugScale=25,DebugSegments=20,DebugThickness=0.25)
    DriverAxis=(X=-1,Y=0,Z=0)
    DriverItem=(Type=Bone,Name="thigh_l")
    FalloffRegionScaleFactors=(PositiveWidth=1,NegativeWidth=1,PositiveHeight=1,NegativeHeight=1)
    FalloffSize=0.45
    OptionalParentItem=(Type=Bone)
    RotationOffset=(X=120,Y=0,Z=0)
   )
   n3 = RigVMVariableNode() {variable=Crouching}
   > &RigUnit_ModifyTransforms(
    ItemToModify="((Item=(Type=Bone,Name=\"dyn_skirt_A_01\"),Transform=(Rotation=(X=-0.2754,Y=0.5606,Z=0.1434,W=0.7677),Translation=(X=0,Y=-2,Z=0),Scale3D=(X=1,Y=1,Z=1))),(Item=(Type=Bone,Name=\"dyn_skirt_A_02\"),Transform=(Rotation=(X=0.059,Y=-0.3504,Z=-0.0592,W=0.9329),Transl ...(truncated)"
    WeightMinimum=0
    Weight=&RigVMFunction_MathDoubleRemap(
     SourceMaximum=1.2
     SourceMinimum=0
     TargetMaximum=1
     TargetMinimum=0
     bClamp=false
     Value=&RigVMFunction_MathDoubleAdd(B=@n2.OutputParam, A=@n3.Value).Result
    ).Result
   )
   &RigUnit_ParentConstraint(
    AdvancedSettings=(InterpolationType=Average,RotationOrderForFilter=XZY)
    Child=(Type=Bone,Name="base_skirt_G")
    Parents=((Item=(Type=Bone,Name="deform_groin_r"),Weight=1))
    Weight=1
    ExecutePin=@SequenceAggregate_3_1_1.B
   )
   > &RigUnit_ParentConstraint(
    AdvancedSettings=(RotationOrderForFilter=XZY)
    Child=(Type=Bone,Name="dyn_skirt_F_01")
    Filter=(TranslationFilter=(bX=False,bY=False,bZ=False),RotationFilter=(bX=True,bY=True,bZ=True),ScaleFilter=(bX=True,bY=True,bZ=True))
    Parents=((Item=(Type=Bone,Name="dyn_skirt_G_01"),Weight=0.5),(Item=(Type=Bone,Name="base_skirt_E"),Weight=0.5))
    Weight=1
   )
   > &RigUnit_ParentConstraint(
    AdvancedSettings=(RotationOrderForFilter=XZY)
    Child=(Type=Bone,Name="base_skirt_E")
    Filter=(TranslationFilter=(bX=False,bY=False,bZ=False),RotationFilter=(bX=True,bY=True,bZ=True),ScaleFilter=(bX=True,bY=True,bZ=True))
    Parents=((Item=(Type=Bone,Name="dyn_skirt_F_01"),Weight=0.5),(Item=(Type=Bone,Name="base_skirt_D"),Weight=0.5))
    Weight=1
   )
   > &RigUnit_ModifyTransforms(
    ItemToModify="((Item=(Type=Bone,Name=\"dyn_skirt_G_01\"),Transform=(Rotation=(X=0,Y=0,Z=-0.0436,W=0.999),Translation=(X=0,Y=0,Z=0),Scale3D=(X=1,Y=1,Z=1))),(Item=(Type=Bone,Name=\"dyn_skirt_F_01\"),Transform=(Rotation=(X=0,Y=0,Z=-0.0436,W=0.999),Translation=(X=0,Y=0,Z=0),Sca ...(truncated)"
    WeightMinimum=0
    Weight=@n1.Result
   )
   > n4 = &RigUnit_SphericalPoseReader(
    ActiveRegionScaleFactors=(PositiveWidth=3.5,NegativeWidth=1,PositiveHeight=3,NegativeHeight=3)
    ActiveRegionSize=0.1
    Debug=(bDrawDebug=false,DebugScale=25,DebugSegments=20,DebugThickness=0.25)
    DriverAxis=(X=1,Y=0,Z=0)
    DriverItem=(Type=Bone,Name="thigh_r")
    FalloffRegionScaleFactors=(PositiveWidth=1,NegativeWidth=1,PositiveHeight=1,NegativeHeight=1)
    FalloffSize=0.45
    FlipHeightScaling=false
    FlipWidthScaling=true
    OptionalParentItem=(Type=Bone)
    RotationOffset=(X=-70,Y=0,Z=0)
   )
   n5 = RigVMVariableNode() {variable=Crouching}
   > &RigUnit_ModifyTransforms(
    ItemToModify="((Item=(Type=Bone,Name=\"dyn_skirt_G_01\"),Transform=(Rotation=(X=0.0058,Y=-0.1525,Z=-0.0377,W=0.9876),Translation=(X=0,Y=0,Z=0),Scale3D=(X=1,Y=1,Z=1))),(Item=(Type=Bone,Name=\"dyn_skirt_G_02\"),Transform=(Rotation=(X=0,Y=0.0082,Z=0.0051,W=1),Translation=(X=0, ...(truncated)"
    WeightMinimum=0
    Weight=&RigVMFunction_MathDoubleRemap(
     SourceMaximum=1.2
     SourceMinimum=0
     TargetMaximum=1
     TargetMinimum=0
     bClamp=false
     Value=&RigVMFunction_MathDoubleAdd(B=@n4.OutputParam, A=@n5.Value).Result
    ).Result
   )
  ]}
 SequenceAggregate_3_1_ContainedGraph {nodes [
   n1 = RigVMFunctionEntryNode()
   SequenceAggregate_3 = &RigVMFunction_Sequence(ExecuteContext=@n1.ExecuteContext)
   SequenceAggregate_3_1 = &RigVMFunction_Sequence(ExecuteContext=@SequenceAggregate_3.B)
   SequenceAggregate_3_2 = &RigVMFunction_Sequence(ExecuteContext=@SequenceAggregate_3_1.B)
   RigVMFunctionReturnNode(A=@SequenceAggregate_3.A, B=@SequenceAggregate_3_1.A, E=@SequenceAggregate_3_2.A, F=@SequenceAggregate_3_2.B)
  ]}
]
hierarchy [
 root {loc=<0 0 0>, rot=<0 0 0 1>}
 pelvis {parent=root, loc=<0 -0.5301 98.0571>, rot=<0.0127 0.707 -0.0127 -0.707>}
 spine_01 {parent=pelvis, loc=<6.1208 -0.9328 0>, rot=<0 0 0.034 -0.9994>}
 spine_02 {parent=spine_01, loc=<6.5996 -0.2095 0>, rot=<0 0 0.0238 -0.9997>}
 spine_03 {parent=spine_02, loc=<6.4914 -0.1237 0>, rot=<0 0 0.0149 -0.9999>}
 spine_04 {parent=spine_03, loc=<6.9598 0.0794 0>, rot=<0 0 -0.0132 -0.9999>}
 spine_05 {parent=spine_04, loc=<6.9987 0 0>, rot=<0 0 -0.0015 -1>}
 attach_cape {parent=spine_05, loc=<9.0155 -11.714 0>, rot=<0.0616 -0.7044 -0.0616 -0.7044>}
 neck_01 {parent=spine_05, loc=<11.3138 0.1529 0>, rot=<0 0 -0.2003 -0.9797>}
 neck_02 {parent=neck_01, loc=<4.5627 0 0>, rot=<0 0 0.0504 -0.9987>}
 head {parent=neck_02, loc=<5.3914 0 0>, rot=<0 0 0.115 -0.9934>}
 attach_backpack {parent=spine_05, loc=<-0.1242 -13.473 0>, rot=<-0.0093 -0.707 0.0093 -0.707>}
 clavicle_l {parent=spine_05, loc=<5.4831 -0.1053 -3.2946>, rot=<0.0465 -0.7289 -0.009 -0.6829>}
 upperarm_l {parent=clavicle_l, loc=<9.8658 -1.2136 -0.4105>, rot=<0.0164 -0.3245 -0.0269 -0.9454>}
 lowerarm_l {parent=upperarm_l, loc=<26.465 -0.8345 0.259>, rot=<0.0036 0.0161 -0.2298 -0.9731>}
 hand_l {parent=lowerarm_l, loc=<22.7355 -0.3203 -0.6105>, rot=<-0.6784 -0.0223 0.0735 -0.7307>}
 thumb_01_l {parent=hand_l, loc=<1.3118 -0.3191 -2.459>, rot=<0.6546 -0.4042 0.0664 -0.6354>}
 thumb_02_l {parent=thumb_01_l, loc=<5.9336 -0.0448 0.0401>, rot=<-0.0037 0.0213 0.298 -0.9543>}
 thumb_03_l {parent=thumb_02_l, loc=<2.2223 0.3682 -0.0907>, rot=<0.0118 -0.013 0.0066 -0.9998>}
 index_metacarpal_l {parent=hand_l, loc=<1.6535 0.0183 -2.96>, rot=<0.0158 -0.0124 -0.0173 -0.9996>}
 index_01_l {parent=index_metacarpal_l, loc=<8.3008 0.0357 -0.0524>, rot=<-0.0044 0.0485 0.0924 -0.9945>}
 index_02_l {parent=index_01_l, loc=<3.3587 -0.3004 0.0452>, rot=<-0.002 0.0027 0.3225 -0.9466>}
 index_03_l {parent=index_02_l, loc=<2.5578 -0.123 -0.0252>, rot=<-0.002 0.0067 0.2558 -0.9667>}
 middle_metacarpal_l {parent=hand_l, loc=<2.1856 -0.1725 -0.5984>, rot=<0.006 0.0168 -0.0329 -0.9993>}
 middle_01_l {parent=middle_metacarpal_l, loc=<7.3275 -0.2447 0.0129>, rot=<-0.0208 0.0042 0.0858 -0.9961>}
 middle_02_l {parent=middle_01_l, loc=<4.0123 0.1401 0.1175>, rot=<-0.0218 0.0258 0.3512 -0.9357>}
 middle_03_l {parent=middle_02_l, loc=<2.6077 0.0534 0.1373>, rot=<0.0082 0.0188 0.1708 -0.9851>}
 ring_metacarpal_l {parent=hand_l, loc=<2.2074 -0.3625 1.3638>, rot=<-0.0284 0.0722 -0.0458 -0.9959>}
 ring_01_l {parent=ring_metacarpal_l, loc=<7.0295 -0.6197 -0.1446>, rot=<-0.0489 -0.0252 0.1892 -0.9804>}
 ring_02_l {parent=ring_01_l, loc=<3.6174 0.4647 -0.1027>, rot=<0 0 0.2806 -0.9598>}
 ring_03_l {parent=ring_02_l, loc=<2.4152 0.114 0.0199>, rot=<0 0 0.1189 -0.9929>}
 pinky_metacarpal_l {parent=hand_l, loc=<1.7616 -0.9753 3.3814>, rot=<-0.0397 0.0814 -0.069 -0.9935>}
 pinky_01_l {parent=pinky_metacarpal_l, loc=<7.1709 -0.4831 -0.1094>, rot=<-0.0654 -0.0035 0.2242 -0.9723>}
 pinky_02_l {parent=pinky_01_l, loc=<2.8274 0.0893 -0.1043>, rot=<0.0037 0.0011 0.2854 -0.9584>}
 pinky_03_l {parent=pinky_02_l, loc=<2.2046 0.0307 -0.0078>, rot=<-0.0002 0.0098 0.1445 -0.9895>}
 deform_wrist_dn_base_l {parent=hand_l, loc=<0 -1.5326 0>, rot=<0 0 0 -1>}
 deform_wrist_dn_l {parent=deform_wrist_dn_base_l, loc=<0 0 0>, rot=<0 0 0 1>}
 deform_wrist_up_base_l {parent=hand_l, loc=<0 2.3009 0>, rot=<0 0 0 -1>}
 deform_wrist_up_l {parent=deform_wrist_up_base_l, loc=<0 0 0>, rot=<0 0 0 1>}
 lowerarm_twist_02_l {parent=lowerarm_l, loc=<10.6707 0 0>, rot=<0 0 0 1>}
 lowerarm_twist_01_l {parent=lowerarm_l, loc=<18.316 0 0>, rot=<0 0 0 1>}
 deform_elbow_in_base_l {parent=lowerarm_l, loc=<0.1834 3.6062 -0.2572>, rot=<0 0 0 -1>}
 deform_elbow_in_l {parent=deform_elbow_in_base_l, loc=<0 0 0>, rot=<0 0 0 1>}
 deform_elbow_out_base_l {parent=lowerarm_l, loc=<0 -4 0>, rot=<0 0 0 1>}
 deform_elbow_out_l {parent=deform_elbow_out_base_l, loc=<0 0 0>, rot=<0 0 0 1>}
 upperarm_twist_01_l {parent=upperarm_l, loc=<7.2135 -0.1134 0>, rot=<0 0 0 1>}
 deform_upperarm_bk_base_l {parent=upperarm_twist_01_l, loc=<-6.4053 -4.3746 0.314>, rot=<0 0 0 -1>}
 deform_upperarm_bk_l {parent=deform_upperarm_bk_base_l, loc=<0 0 0>, rot=<0 0 0 1>}
 deform_upperarm_fr_base_l {parent=upperarm_twist_01_l, loc=<-5.6312 5.092 0.3354>, rot=<0 0 0 -1>}
 deform_upperarm_fr_l {parent=deform_upperarm_fr_base_l, loc=<0 0 0>, rot=<0 0 0 1>}
 deform_upperarm_up_base_l {parent=upperarm_twist_01_l, loc=<-8.891 0.4825 3.9517>, rot=<0 0 0 -1>}
 deform_upperarm_up_l {parent=deform_upperarm_up_base_l, loc=<0 0 0>, rot=<0 0 0 1>}
 base_sleeve_A_l {parent=upperarm_twist_01_l, loc=<0.7416 -0.4507 3.7392>, rot=<-0.0006 0.0659 0.0088 -0.9978>}
 dyn_sleeve_A_l {parent=base_sleeve_A_l, loc=<0 0 0>, rot=<0 0 0 1>}
 base_sleeve_D_l {parent=upperarm_twist_01_l, loc=<1.1453 -4.7281 0.4456>, rot=<-0.7329 0.013 0.0108 -0.6802>}
 dyn_sleeve_D_l {parent=base_sleeve_D_l, loc=<0 0 0>, rot=<0 0 0 1>}
 base_sleeve_B_l {parent=upperarm_twist_01_l, loc=<1.5131 3.8487 -1.4985>, rot=<0.7096 0.0401 0.0531 -0.7014>}
 dyn_sleeve_B_l {parent=base_sleeve_B_l, loc=<0 0 0>, rot=<0 0 0 1>}
 base_sleeve_C_l {parent=upperarm_twist_01_l, loc=<1.9216 -0.4118 -4.904>, rot=<0.9998 -0.0169 0.001 -0.0082>}
 dyn_sleeve_C_l {parent=base_sleeve_C_l, loc=<0 0 0>, rot=<0 0 0 1>}
 upperarm_twist_02_l {parent=upperarm_l, loc=<19.7998 -0.1134 0>, rot=<0 0 0 1>}
 clavicle_r {parent=spine_05, loc=<5.4831 -0.1053 3.2946>, rot=<0.7289 0.0465 -0.6829 0.009>}
 upperarm_r {parent=clavicle_r, loc=<-9.8658 1.2136 0.4105>, rot=<0.0164 -0.3245 -0.0269 -0.9454>}
 lowerarm_r {parent=upperarm_r, loc=<-26.465 0.8345 -0.259>, rot=<0.0036 0.0161 -0.2298 -0.9731>}
 hand_r {parent=lowerarm_r, loc=<-22.7355 0.3203 0.6106>, rot=<-0.6784 -0.0223 0.0735 -0.7307>}
 thumb_01_r {parent=hand_r, loc=<-1.3118 0.3191 2.459>, rot=<0.6546 -0.4042 0.0664 -0.6354>}
 thumb_02_r {parent=thumb_01_r, loc=<-5.9336 0.0448 -0.0401>, rot=<-0.0037 0.0213 0.298 -0.9543>}
 thumb_03_r {parent=thumb_02_r, loc=<-2.2223 -0.3682 0.0907>, rot=<0.0118 -0.013 0.0066 -0.9998>}
 index_metacarpal_r {parent=hand_r, loc=<-1.6535 -0.0183 2.96>, rot=<0.0158 -0.0124 -0.0173 -0.9996>}
 index_01_r {parent=index_metacarpal_r, loc=<-8.3008 -0.0357 0.0524>, rot=<-0.0044 0.0485 0.0924 -0.9945>}
 index_02_r {parent=index_01_r, loc=<-3.3587 0.3004 -0.0452>, rot=<-0.002 0.0027 0.3225 -0.9466>}
 index_03_r {parent=index_02_r, loc=<-2.5578 0.123 0.0252>, rot=<-0.002 0.0067 0.2558 -0.9667>}
 middle_metacarpal_r {parent=hand_r, loc=<-2.1856 0.1725 0.5984>, rot=<0.006 0.0168 -0.0329 -0.9993>}
 middle_01_r {parent=middle_metacarpal_r, loc=<-7.3275 0.2447 -0.0129>, rot=<-0.0208 0.0042 0.0858 -0.9961>}
 middle_02_r {parent=middle_01_r, loc=<-4.0123 -0.1401 -0.1175>, rot=<-0.0218 0.0258 0.3512 -0.9357>}
 middle_03_r {parent=middle_02_r, loc=<-2.6077 -0.0534 -0.1373>, rot=<0.0082 0.0188 0.1708 -0.9851>}
 ring_metacarpal_r {parent=hand_r, loc=<-2.2074 0.3625 -1.3638>, rot=<-0.0284 0.0722 -0.0458 -0.9959>}
 ring_01_r {parent=ring_metacarpal_r, loc=<-7.0295 0.6197 0.1446>, rot=<-0.0489 -0.0252 0.1892 -0.9804>}
 ring_02_r {parent=ring_01_r, loc=<-3.6174 -0.4647 0.1027>, rot=<0 0 0.2806 -0.9598>}
 ring_03_r {parent=ring_02_r, loc=<-2.4152 -0.114 -0.0199>, rot=<0 0 0.1189 -0.9929>}
 pinky_metacarpal_r {parent=hand_r, loc=<-1.7616 0.9753 -3.3814>, rot=<-0.0397 0.0814 -0.069 -0.9935>}
 pinky_01_r {parent=pinky_metacarpal_r, loc=<-7.1709 0.4831 0.1094>, rot=<-0.0654 -0.0035 0.2242 -0.9723>}
 pinky_02_r {parent=pinky_01_r, loc=<-2.8274 -0.0893 0.1043>, rot=<0.0037 0.0011 0.2854 -0.9584>}
 pinky_03_r {parent=pinky_02_r, loc=<-2.2046 -0.0307 0.0078>, rot=<-0.0002 0.0098 0.1445 -0.9895>}
 deform_wrist_dn_base_r {parent=hand_r, loc=<-0.0001 1.5328 0>, rot=<0 0 0 -1>}
 deform_wrist_dn_r {parent=deform_wrist_dn_base_r, loc=<0 0 0>, rot=<0 0 0 1>}
 deform_wrist_up_base_r {parent=hand_r, loc=<-0.0002 -2.3007 -0.0001>, rot=<0 0 0 -1>}
 deform_wrist_up_r {parent=deform_wrist_up_base_r, loc=<0 0 0>, rot=<0 0 0 1>}
 lowerarm_twist_02_r {parent=lowerarm_r, loc=<-10.6707 0 0>, rot=<0 0 0 -1>}
 lowerarm_twist_01_r {parent=lowerarm_r, loc=<-18.316 0 0>, rot=<0 0 0 -1>}
 deform_elbow_in_base_r {parent=lowerarm_r, loc=<-0.1834 -3.6062 0.2572>, rot=<0 0 0 -1>}
 deform_elbow_in_r {parent=deform_elbow_in_base_r, loc=<0 0 0>, rot=<0 0 0 1>}
 deform_elbow_out_base_r {parent=lowerarm_r, loc=<0 4 -0.0001>, rot=<0 0 0 -1>}
 deform_elbow_out_r {parent=deform_elbow_out_base_r, loc=<0 0 0>, rot=<0 0 0 1>}
 upperarm_twist_01_r {parent=upperarm_r, loc=<-7.2135 0.1134 0>, rot=<0 0 0 -1>}
 deform_upperarm_bk_base_r {parent=upperarm_twist_01_r, loc=<6.4053 4.3746 -0.314>, rot=<0 0 0 -1>}
 deform_upperarm_bk_r {parent=deform_upperarm_bk_base_r, loc=<0 0 0>, rot=<0 0 0 1>}
 deform_upperarm_fr_base_r {parent=upperarm_twist_01_r, loc=<5.6312 -5.092 -0.3354>, rot=<0 0 0 -1>}
 deform_upperarm_fr_r {parent=deform_upperarm_fr_base_r, loc=<0 0 0>, rot=<0 0 0 1>}
 deform_upperarm_up_base_r {parent=upperarm_twist_01_r, loc=<8.891 -0.4825 -3.9517>, rot=<0 0 0 -1>}
 deform_upperarm_up_r {parent=deform_upperarm_up_base_r, loc=<0 0 0>, rot=<0 0 0 1>}
 base_sleeve_A_r {parent=upperarm_twist_01_r, loc=<-0.7415 0.4507 -3.7394>, rot=<-0.018 -0.9998 0.002 -0.0076>}
 dyn_sleeve_A_r {parent=base_sleeve_A_r, loc=<0 0 0>, rot=<0 0 0 1>}
 base_sleeve_B_r {parent=upperarm_twist_01_r, loc=<-1.6979 -3.7362 1.3144>, rot=<-0.0528 -0.7414 0.6687 -0.0206>}
 dyn_sleeve_B_r {parent=base_sleeve_B_r, loc=<0 0 0>, rot=<0 0 0 1>}
 base_sleeve_C_r {parent=upperarm_twist_01_r, loc=<-1.9055 -0.1343 5.1588>, rot=<-0.0526 -0.0843 0.9948 0.021>}
 dyn_sleeve_C_r {parent=base_sleeve_C_r, loc=<0 0 0>, rot=<0 0 0 1>}
 base_sleeve_D_r {parent=upperarm_twist_01_r, loc=<-1.635 4.6072 2.3783>, rot=<0.0066 0.6769 0.7358 0.0184>}
 dyn_sleeve_D_r {parent=base_sleeve_D_r, loc=<0 0 0>, rot=<0 0 0 1>}
 upperarm_twist_02_r {parent=upperarm_r, loc=<-19.7998 0.1134 0>, rot=<0 0 0 -1>}
 deform_pec_aim_base_l {parent=spine_04, loc=<3.2456 9.3309 -1.2237>, rot=<0.0603 -0.4033 0.3318 -0.8507>}
 deform_pec_aim_l {parent=deform_pec_aim_base_l, loc=<0 0 0>, rot=<0 0 0 1>}
 deform_pec_base_l {parent=deform_pec_aim_l, loc=<9.3665 0 0>, rot=<0 0 0 1>}
 deform_pec_l {parent=deform_pec_base_l, loc=<0 0 0>, rot=<0 0 0 1>}
 deform_pec_aim_base_r {parent=spine_04, loc=<3.2452 9.3309 1.2237>, rot=<-0.4033 -0.0603 0.8507 0.3318>}
 deform_pec_aim_r {parent=deform_pec_aim_base_r, loc=<0 0 0>, rot=<0 0 0 1>}
 deform_pec_base_r {parent=deform_pec_aim_r, loc=<-9.3667 -0.0003 -0.0004>, rot=<0 0 0 1>}
 deform_pec_r {parent=deform_pec_base_r, loc=<0 0 0>, rot=<0 0 0 1>}
 thigh_l {parent=pelvis, loc=<-1.0855 -0.2109 -8.7053>, rot=<0.0009 0.0355 0.0155 -0.9993>}
 calf_l {parent=thigh_l, loc=<-42.5284 -0.0019 -0.7297>, rot=<0.0006 -0.0161 -0.0391 -0.9991>}
 foot_l {parent=calf_l, loc=<-46.7401 -2.5014 -0.592>, rot=<0.0427 -0.0238 0.0353 -0.9982>}
 ball_l {parent=foot_l, loc=<-3.5658 10.8066 -0.1544>, rot=<0.0001 -0.0001 -0.7063 -0.7079>}
 calf_twist_02_l {parent=calf_l, loc=<-19.1546 -1.7943 -0.2638>, rot=<0.0237 0.0095 -0.0035 -0.9997>}
 calf_twist_01_l {parent=calf_l, loc=<-35.8483 -1.9197 -0.5794>, rot=<0.0237 0.0095 -0.0035 -0.9997>}
 deform_knee_in_base_l {parent=calf_l, loc=<0.8031 -5.0231 0.2456>, rot=<0 0 0 -1>}
 deform_knee_in_l {parent=deform_knee_in_base_l, loc=<0 0 0>, rot=<0 0 0 1>}
 deform_knee_out_base_l {parent=calf_l, loc=<0 4 0>, rot=<0 0 0 -1>}
 deform_knee_out_l {parent=deform_knee_out_base_l, loc=<0 0 0>, rot=<0 0 0 1>}
 thigh_twist_01_l {parent=thigh_l, loc=<-27.952 0 0>, rot=<0.0013 0 0 -1>}
 base_legPouch_l {parent=thigh_twist_01_l, loc=<-1.0646 -2.5887 -6.0788>, rot=<0.0629 0.0151 -0.0215 -0.9977>}
 dyn_legPouch_l {parent=base_legPouch_l, loc=<0 0 0>, rot=<0 0 0 1>}
 base_strapFront_l {parent=dyn_legPouch_l, loc=<-2.4391 4.1722 -3.5848>, rot=<0 0.0236 0 -0.9997>}
 dyn_strapFront_1_l {parent=base_strapFront_l, loc=<0 0 0>, rot=<0 0 0 1>}
 dyn_strapFront_2_l {parent=dyn_strapFront_1_l, loc=<-4.191 0 0>, rot=<0 0 0 1>}
 dyn_strapFront_3_l {parent=dyn_strapFront_2_l, loc=<-4.1563 0 0>, rot=<0 0 0 1>}
 base_strapBack_l {parent=dyn_legPouch_l, loc=<-2.472 -5.2492 -2.0578>, rot=<0.1691 0.0241 0.0041 -0.9853>}
 dyn_strapBack_1_l {parent=base_strapBack_l, loc=<0 0 0>, rot=<0 0 0 1>}
 dyn_strapBack_2_l {parent=dyn_strapBack_1_l, loc=<-4.2081 0 0>, rot=<0 0 0 1>}
 dyn_strapBack_3_l {parent=dyn_strapBack_2_l, loc=<-4.1563 0 0>, rot=<0 0 0 1>}
 deform_groin_base_l {parent=thigh_l, loc=<0.4801 5.6776 -0.2486>, rot=<0 0 0 -1>}
 deform_groin_l {parent=deform_groin_base_l, loc=<0 0 0>, rot=<0 0 0 1>}
 thigh_r {parent=pelvis, loc=<-1.0855 -0.2109 8.7053>, rot=<0.0355 -0.0009 0.9993 0.0155>}
 calf_r {parent=thigh_r, loc=<42.5284 0.0019 0.7297>, rot=<0.0006 -0.0161 -0.0391 -0.9991>}
 foot_r {parent=calf_r, loc=<46.7401 2.5014 0.592>, rot=<0.0427 -0.0238 0.0353 -0.9982>}
 ball_r {parent=foot_r, loc=<3.5658 -10.8066 0.1544>, rot=<0.0001 -0.0001 -0.7063 -0.7079>}
 calf_twist_02_r {parent=calf_r, loc=<19.1546 1.7943 0.2638>, rot=<0.0237 0.0095 -0.0035 -0.9997>}
 calf_twist_01_r {parent=calf_r, loc=<35.8483 1.9197 0.5794>, rot=<0.0237 0.0095 -0.0035 -0.9997>}
 deform_knee_in_base_r {parent=calf_r, loc=<-0.8031 5.0231 -0.2457>, rot=<0 0 0 -1>}
 deform_knee_in_r {parent=deform_knee_in_base_r, loc=<0 0 0>, rot=<0 0 0 1>}
 deform_knee_out_base_r {parent=calf_r, loc=<0 -4 0>, rot=<0 0 0 -1>}
 deform_knee_out_r {parent=deform_knee_out_base_r, loc=<0 0 0>, rot=<0 0 0 1>}
 thigh_twist_01_r {parent=thigh_r, loc=<27.952 0 0>, rot=<0.0013 0 0 -1>}
 base_legPouch_r {parent=thigh_twist_01_r, loc=<1.1867 2.6331 6.0983>, rot=<0.2307 0.0186 -0.0187 -0.9727>}
 dyn_legPouch_r {parent=base_legPouch_r, loc=<0 0 0>, rot=<0 0 0 -1>}
 base_strapFront_r {parent=dyn_legPouch_r, loc=<2.3146 -5.1563 1.9566>, rot=<-0.1691 0.0232 -0.004 -0.9853>}
 dyn_strapFront_1_r {parent=base_strapFront_r, loc=<0 0 0>, rot=<0 0 0 1>}
 dyn_strapFront_2_r {parent=dyn_strapFront_1_r, loc=<4.191 0 0>, rot=<0 0 0 -1>}
 dyn_strapFront_3_r {parent=dyn_strapFront_2_r, loc=<4.1563 0 0.0001>, rot=<0 0 0 -1>}
 base_strapBack_r {parent=dyn_legPouch_r, loc=<2.3476 4.2351 3.6586>, rot=<0 0.0218 0 -0.9998>}
 dyn_strapBack_1_r {parent=base_strapBack_r, loc=<0 0 0>, rot=<0 0 0 1>}
 dyn_strapBack_2_r {parent=dyn_strapBack_1_r, loc=<4.2082 0 0>, rot=<0 0 0 -1>}
 dyn_strapBack_3_r {parent=dyn_strapBack_2_r, loc=<4.1562 0 0>, rot=<0 0 0 -1>}
 deform_groin_base_r {parent=thigh_r, loc=<-0.4801 -5.6776 0.2486>, rot=<0 0 0 -1>}
 deform_groin_r {parent=deform_groin_base_r, loc=<0 0 0>, rot=<0 0 0 1>}
 deform_glute_base_l {parent=pelvis, loc=<-7.3466 -9.0213 -9.1245>, rot=<0.0172 0.031 -0.4477 -0.8935>}
 deform_glute_l {parent=deform_glute_base_l, loc=<0 0 0>, rot=<0 0 0 1>}
 deform_glute_base_r {parent=pelvis, loc=<-7.3464 -9.0213 9.1245>, rot=<0.031 -0.0172 0.8935 -0.4477>}
 deform_glute_r {parent=deform_glute_base_r, loc=<0 0 0>, rot=<0 0 0 1>}
 base_master_sleeve {parent=pelvis, loc=<0 8.4001 0>, rot=<0 0 0 1>}
 dyn_master_sleeve {parent=base_master_sleeve, loc=<0 0 0>, rot=<0 0 0 1>}
 base_sleeve_l {parent=dyn_master_sleeve, loc=<-2.1586 0.6226 -7.1932>, rot=<0.0255 -0.7941 -0.5996 -0.0962>}
 dyn_sleeve_l_1 {parent=base_sleeve_l, loc=<0 0 0>, rot=<0 0 0 1>}
 dyn_sleeve_l_2 {parent=dyn_sleeve_l_1, loc=<8.6439 0.1076 0>, rot=<0 0 0.017 -0.9999>}
 dyn_sleeve_l_3 {parent=dyn_sleeve_l_2, loc=<8.169 -0.0703 0>, rot=<0 0 -0.0111 -0.9999>}
 dyn_sleeve_l_4 {parent=dyn_sleeve_l_3, loc=<8.9367 0.1306 0>, rot=<0 0 0.0159 -0.9999>}
 base_sleeve_r {parent=dyn_master_sleeve, loc=<-1.0654 3.4327 3.5372>, rot=<0.1443 0.6042 0.7833 -0.0242>}
 dyn_sleeve_r_1 {parent=base_sleeve_r, loc=<0 0 0>, rot=<0 0 0 1>}
 dyn_sleeve_r_2 {parent=dyn_sleeve_r_1, loc=<5.8555 0.2133 0>, rot=<0 0 -0.0324 -0.9995>}
 dyn_sleeve_r_3 {parent=dyn_sleeve_r_2, loc=<5.3797 -0.1324 0>, rot=<0 0 -0.0111 -0.9999>}
 base_skirt_A {parent=pelvis, loc=<-0.4568 5.055 -14.3572>, rot=<-0.1227 0.6517 0.7462 0.0581>}
 dyn_skirt_A_01 {parent=base_skirt_A, loc=<0 0 0>, rot=<0 0 0 1>}
 dyn_skirt_A_02 {parent=dyn_skirt_A_01, loc=<11.1309 -0.3795 0.1239>, rot=<-0.0199 0.0038 0.0129 -0.9997>}
 dyn_skirt_A_03 {parent=dyn_skirt_A_02, loc=<7.9482 -0.0752 0>, rot=<0 0 -0.0111 -0.9999>}
 dyn_skirt_A_04 {parent=dyn_skirt_A_03, loc=<9.6761 0.1072 0>, rot=<0 0 0.0159 -0.9999>}
 base_skirt_B {parent=pelvis, loc=<0.8813 -3.544 -14.2927>, rot=<0.0938 -0.8273 -0.5514 -0.0522>}
 dyn_skirt_B_01 {parent=base_skirt_B, loc=<0 0 0>, rot=<0 0 0 1>}
 dyn_skirt_B_02 {parent=dyn_skirt_B_01, loc=<9.9533 0.1061 0>, rot=<0 0 0.0287 -0.9996>}
 dyn_skirt_B_03 {parent=dyn_skirt_B_02, loc=<9.4768 -0.0412 0>, rot=<0 0 -0.0111 -0.9999>}
 dyn_skirt_B_04 {parent=dyn_skirt_B_03, loc=<11.2043 0.0587 0>, rot=<0 0 0.0159 -0.9999>}
 base_skirt_C {parent=pelvis, loc=<0.1412 -10.0792 -8.1491>, rot=<0.0705 -0.9919 -0.069 -0.0804>}
 dyn_skirt_C_01 {parent=base_skirt_C, loc=<0 0 0>, rot=<0 0 0 1>}
 dyn_skirt_C_02 {parent=dyn_skirt_C_01, loc=<9.9533 0.1061 0>, rot=<0 0 -0.0054 -1>}
 dyn_skirt_C_03 {parent=dyn_skirt_C_02, loc=<9.4768 -0.0412 0>, rot=<0 0 0.0037 -1>}
 dyn_skirt_C_04 {parent=dyn_skirt_C_03, loc=<11.2043 0.0587 0>, rot=<0 0 0.0159 -0.9999>}
 base_skirt_D {parent=pelvis, loc=<-0.5378 -12.3016 0>, rot=<0.1201 -0.9928 0 0>}
 dyn_skirt_D_01 {parent=base_skirt_D, loc=<0 0 0>, rot=<0 0 0 1>}
 dyn_skirt_D_02 {parent=dyn_skirt_D_01, loc=<10.6793 -0.0887 0>, rot=<0 0 -0.0394 -0.9992>}
 dyn_skirt_D_03 {parent=dyn_skirt_D_02, loc=<10.2021 -0.0251 0>, rot=<0 0 -0.0111 -0.9999>}
 dyn_skirt_D_04 {parent=dyn_skirt_D_03, loc=<11.9295 0.0357 0>, rot=<0 0 0.0159 -0.9999>}
 base_skirt_E {parent=pelvis, loc=<0.1224 -10.7002 7.9931>, rot=<0.1044 -0.9863 0.1074 0.0691>}
 dyn_skirt_E_01 {parent=base_skirt_E, loc=<0 0 0>, rot=<0 0 0 1>}
 dyn_skirt_E_02 {parent=dyn_skirt_E_01, loc=<9.6309 0.1246 0>, rot=<0 0 -0.0106 -0.9999>}
 dyn_skirt_E_03 {parent=dyn_skirt_E_02, loc=<9.154 -0.0484 0>, rot=<0 0 -0.0111 -0.9999>}
 dyn_skirt_E_04 {parent=dyn_skirt_E_03, loc=<10.8816 0.0689 0>, rot=<0 0 0.0159 -0.9999>}
 base_skirt_F {parent=pelvis, loc=<1.0153 -3.2578 13.4888>, rot=<0.0923 -0.7755 0.62 0.0749>}
 dyn_skirt_F_01 {parent=base_skirt_F, loc=<0 0 0>, rot=<0 0 0 1>}
 dyn_skirt_F_02 {parent=dyn_skirt_F_01, loc=<9.6309 0.1246 0>, rot=<0 0 0.0287 -0.9996>}
 dyn_skirt_F_03 {parent=dyn_skirt_F_02, loc=<9.154 -0.0484 0>, rot=<0 0 -0.0111 -0.9999>}
 dyn_skirt_F_04 {parent=dyn_skirt_F_03, loc=<10.8816 0.0689 0>, rot=<0 0 0.0159 -0.9999>}
 base_skirt_G {parent=pelvis, loc=<0.8395 5.0369 13.5807>, rot=<0.1366 -0.5504 0.82 0.0772>}
 dyn_skirt_G_01 {parent=base_skirt_G, loc=<0 0 0>, rot=<0 0 0 1>}
 dyn_skirt_G_02 {parent=dyn_skirt_G_01, loc=<11.2796 0.131 -0.1313>, rot=<0.0009 0.0006 -0.0311 -0.9995>}
 dyn_skirt_G_03 {parent=dyn_skirt_G_02, loc=<8.0835 -0.0383 0>, rot=<0 0 0.0116 -0.9999>}
 dyn_skirt_G_04 {parent=dyn_skirt_G_03, loc=<9.8114 0.1029 0>, rot=<0 0 0.0159 -0.9999>}
 base_jacket_helper_sleeve_r {parent=pelvis, loc=<-0.6434 8.0193 8.947>, rot=<0.032 -0.1873 0.9815 0.0218>}
 dyn_jacket_helper_sleeve_r {parent=base_jacket_helper_sleeve_r, loc=<0 0 0>, rot=<0 0 0 1>}
 base_jacket_helper_sleeve_l {parent=pelvis, loc=<-1.166 6.4529 -11.6235>, rot=<0.2656 0.0301 0.0243 -0.9633>}
 dyn_jacket_helper_sleeve_l {parent=base_jacket_helper_sleeve_l, loc=<0 0 0>, rot=<0 0 0 1>}
]
variables [
 Crouching {type=double, default=0}
 FrontEndAlpha {type=double, default=0}
 SkydiveMode {type=double, default=0}
 TacticalJumping {type=double, default=0}
]
