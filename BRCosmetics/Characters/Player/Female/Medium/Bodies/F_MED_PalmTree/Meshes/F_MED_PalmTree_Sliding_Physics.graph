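# Physics asset for the sliding variant of the female medium PalmTree cosmetic body, defining capsule bodies for its dangling skirt and sleeves.
# Seven four-link skirt chains (A–G) and two sleeve chains (left four links, right three) hang from small base anchors.
# Their joints have per-chain swing limits and near-zero twist.
# Anchors and two thigh capsules link freely to root, which carries one oversized, offset, near-spherical capsule.
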PhysicsAsset /BRCosmetics/Characters/Player/Female/Medium/Bodies/F_MED_PalmTree/Meshes/F_MED_PalmTree_Sliding_Physics
bodies {
 base_skirt_A [{shapeType=Capsule, length=0.505, radius=0.505, shapeName=base_skirt_A_capsule, rotation=<rot 0 -90 0>}]
 base_skirt_B [{shapeType=Capsule, length=0.505, radius=0.505, shapeName=base_skirt_B_capsule, rotation=<rot 0 -90 0>}]
 base_skirt_C [{shapeType=Capsule, length=0.505, radius=0.505, shapeName=base_skirt_C_capsule, rotation=<rot 0 -90 0>}]
 base_skirt_D [{shapeType=Capsule, length=0.505, radius=0.505, shapeName=base_skirt_D_capsule, rotation=<rot 0 -90 0>}]
 base_skirt_E [{shapeType=Capsule, length=0.505, radius=0.505, shapeName=base_skirt_E_capsule, rotation=<rot 0 -90 0>}]
 base_skirt_F [{shapeType=Capsule, length=0.505, radius=0.505, shapeName=base_skirt_F_capsule, rotation=<rot 0 -90 0>}]
 base_skirt_G [{shapeType=Capsule, length=0.505, radius=0.505, shapeName=base_skirt_G_capsule, rotation=<rot 0 -90 0>}]
 base_sleeve_l [{shapeType=Capsule, length=0.505, radius=0.505, shapeName=base_sleeve_l_capsule, rotation=<rot 0 -90 0>}]
 base_sleeve_r [{shapeType=Capsule, length=0.505, radius=0.505, shapeName=base_sleeve_r_capsule, rotation=<rot 0 -90 0>}]
 dyn_skirt_A_01 [{shapeType=Capsule, length=2.2783, radius=1.5257, shapeName=dyn_skirt_A_01_capsule, center=<5.4003 0 0>, rotation=<rot 0 -88.6045 -81.1754>}]
 dyn_skirt_A_02 [{shapeType=Capsule, length=2.8125, radius=2.8125, shapeName=dyn_skirt_A_02_capsule, center=<4.4472 0 0>, rotation=<rot 90 0 0>}]
 dyn_skirt_A_03 [{shapeType=Capsule, length=2.375, radius=3.25, shapeName=dyn_skirt_A_03_capsule, center=<4.4472 0 0>, rotation=<rot 90 0 0>}]
 dyn_skirt_A_04 [{shapeType=Capsule, length=2.0625, radius=2.875, shapeName=dyn_skirt_A_04_capsule, center=<3.1795 0 0>, rotation=<rot -90 10.0249 -10.0251>}]
 dyn_skirt_B_01 [{shapeType=Capsule, length=4.625, radius=2.6875, shapeName=dyn_skirt_B_01_capsule, center=<2.2245 0 0>}]
 dyn_skirt_B_02 [{shapeType=Capsule, length=4.9375, radius=2.375, shapeName=dyn_skirt_B_02_capsule, center=<2.2245 0 0>}]
 dyn_skirt_B_03 [{shapeType=Capsule, length=4.25, radius=3.0625, shapeName=dyn_skirt_B_03_capsule, center=<2.2245 0 0>}]
 dyn_skirt_B_04 [{shapeType=Capsule, length=4.25, radius=3.0625, shapeName=dyn_skirt_B_04_capsule, center=<2.2245 0 0>}]
 dyn_skirt_C_01 [{shapeType=Capsule, length=4.75, radius=2.5625, shapeName=dyn_skirt_C_01_capsule, center=<2.5007 0 0>}]
 dyn_skirt_C_02 [{shapeType=Capsule, length=6, radius=2.5625, shapeName=dyn_skirt_C_02_capsule, center=<2.5007 0 0>}]
 dyn_skirt_C_03 [{shapeType=Capsule, length=6, radius=2.5625, shapeName=dyn_skirt_C_03_capsule, center=<2.5007 0 0>}]
 dyn_skirt_C_04 [{shapeType=Capsule, length=6, radius=2.5625, shapeName=dyn_skirt_C_04_capsule, center=<2.5007 0 0>}]
 dyn_skirt_D_01 [{shapeType=Capsule, length=3.6875, radius=2.625, shapeName=dyn_skirt_D_01_capsule, center=<3.0274 0 0>}]
 dyn_skirt_D_02 [{shapeType=Capsule, length=5, radius=3.4375, shapeName=dyn_skirt_D_02_capsule, center=<3.6744 0 0>}]
 dyn_skirt_D_03 [{shapeType=Capsule, length=5, radius=3.4375, shapeName=dyn_skirt_D_03_capsule, center=<3.6744 0 0>}]
 dyn_skirt_D_04 [{shapeType=Capsule, length=5, radius=3.4375, shapeName=dyn_skirt_D_04_capsule, center=<3.6744 0 0>}]
 dyn_skirt_E_01 [{shapeType=Capsule, length=4.75, radius=2.5625, shapeName=dyn_skirt_E_01_capsule, center=<2.5007 0 0>}]
 dyn_skirt_E_02 [{shapeType=Capsule, length=6, radius=2.5625, shapeName=dyn_skirt_E_02_capsule, center=<2.5007 0 0>}]
 dyn_skirt_E_03 [{shapeType=Capsule, length=6, radius=2.5625, shapeName=dyn_skirt_E_03_capsule, center=<2.5007 0 0>}]
 dyn_skirt_E_04 [{shapeType=Capsule, length=6, radius=2.5625, shapeName=dyn_skirt_E_04_capsule, center=<2.5007 0 0>}]
 dyn_skirt_F_01 [{shapeType=Capsule, length=4.625, radius=2.6875, shapeName=dyn_skirt_F_01_capsule, center=<2.2245 0 0>}]
 dyn_skirt_F_02 [{shapeType=Capsule, length=4.9375, radius=2.375, shapeName=dyn_skirt_F_02_capsule, center=<2.2245 0 0>}]
 dyn_skirt_F_03 [{shapeType=Capsule, length=4.25, radius=3.0625, shapeName=dyn_skirt_F_03_capsule, center=<2.2245 0 0>}]
 dyn_skirt_F_04 [{shapeType=Capsule, length=4.25, radius=3.0625, shapeName=dyn_skirt_F_04_capsule, center=<2.2245 0 0>}]
 dyn_skirt_G_01 [{shapeType=Capsule, length=2.2783, radius=1.5257, shapeName=dyn_skirt_G_01_capsule, center=<5.8414 0 0>, rotation=<rot 90 0 0>}]
 dyn_skirt_G_02 [{shapeType=Capsule, length=2.8125, radius=2.8125, shapeName=dyn_skirt_G_02_capsule, center=<4.4472 0 0>, rotation=<rot 90 0 0>}]
 dyn_skirt_G_03 [{shapeType=Capsule, length=2.8125, radius=2.8125, shapeName=dyn_skirt_G_03_capsule, center=<4.4472 0 0>, rotation=<rot 90 0 0>}]
 dyn_skirt_G_04 [{shapeType=Capsule, length=2.375, radius=3.25, shapeName=dyn_skirt_G_04_capsule, center=<4.4472 0 0>, rotation=<rot 90 0 0>}]
 dyn_sleeve_l_1 [{shapeType=Capsule, length=4.625, radius=1.6875, shapeName=dyn_sleeve_l_1_capsule, center=<2.8014 0 0>, rotation=<rot 0.0004 0.0001 97.2388>}]
 dyn_sleeve_l_2 [{shapeType=Capsule, length=4.625, radius=1.6875, shapeName=dyn_sleeve_l_2_capsule, center=<2.8014 0 0>, rotation=<rot 0.0004 0.0001 97.2388>}]
 dyn_sleeve_l_3 [{shapeType=Capsule, length=4.625, radius=1.6875, shapeName=dyn_sleeve_l_3_capsule, center=<2.8014 0 0>, rotation=<rot 0.0004 0.0001 97.2388>}]
 dyn_sleeve_l_4 [{shapeType=Capsule, length=3.75, radius=2.5625, shapeName=dyn_sleeve_l_4_capsule, center=<2.8014 0 0>, rotation=<rot 0.0004 0.0001 97.2388>}]
 dyn_sleeve_r_1 [{shapeType=Capsule, length=4.625, radius=1.6875, shapeName=dyn_sleeve_r_1_capsule, center=<2.8014 0 0>, rotation=<rot 0.0001 0 85.9214>}]
 dyn_sleeve_r_2 [{shapeType=Capsule, length=4.625, radius=1.6875, shapeName=dyn_sleeve_r_2_capsule, center=<2.8014 0 0>, rotation=<rot 0.0001 0 85.9214>}]
 dyn_sleeve_r_3 [{shapeType=Capsule, length=3.625, radius=2.6875, shapeName=dyn_sleeve_r_3_capsule, center=<5.3886 0 -0.4301>, rotation=<rot 0.0001 0 85.9214>}]
 root [{shapeType=Capsule, length=4, radius=70.0355, shapeName=root_capsule, center=<0 -6.573 -57.7941>}]
 thigh_l [{
   shapeType=Capsule
   length=25.5767
   radius=9.1106
   shapeName=thigh_l_capsule
   center=<-23.5094 -0.7302 1.5453>
   rotation=<rot 0 -91.9195 -90.5279>
  }]
 thigh_r [{
   shapeType=Capsule
   length=24.7017
   radius=9.11
   shapeName=thigh_l_capsule
   center=<23.5094 0.7302 -1.5453>
   rotation=<rot 0 -91.9195 89.4721>
  }]
}
constraints [
 {
  bone1Name=thigh_l
  bone2Name=root
  swing1LimitDegrees=45
  swing1Motion=Free
  swing2LimitDegrees=45
  swing2Motion=Free
  twistLimitDegrees=45
  twistMotion=Free
 }
 {
  bone1Name=thigh_r
  bone2Name=root
  swing1LimitDegrees=45
  swing1Motion=Free
  swing2LimitDegrees=45
  swing2Motion=Free
  twistLimitDegrees=45
  twistMotion=Free
 }
 {
  bone1Name=base_skirt_C
  bone2Name=root
  swing1LimitDegrees=45
  swing1Motion=Free
  swing2LimitDegrees=45
  swing2Motion=Free
  twistLimitDegrees=45
  twistMotion=Free
 }
 {
  bone1Name=base_skirt_G
  bone2Name=root
  swing1LimitDegrees=45
  swing1Motion=Free
  swing2LimitDegrees=45
  swing2Motion=Free
  twistLimitDegrees=45
  twistMotion=Free
 }
 {
  bone1Name=base_skirt_F
  bone2Name=root
  swing1LimitDegrees=45
  swing1Motion=Free
  swing2LimitDegrees=45
  swing2Motion=Free
  twistLimitDegrees=45
  twistMotion=Free
 }
 {
  bone1Name=base_skirt_E
  bone2Name=root
  swing1LimitDegrees=45
  swing1Motion=Free
  swing2LimitDegrees=45
  swing2Motion=Free
  twistLimitDegrees=45
  twistMotion=Free
 }
 {
  bone1Name=base_skirt_D
  bone2Name=root
  swing1LimitDegrees=45
  swing1Motion=Free
  swing2LimitDegrees=45
  swing2Motion=Free
  twistLimitDegrees=45
  twistMotion=Free
 }
 {
  bone1Name=base_skirt_B
  bone2Name=root
  swing1LimitDegrees=45
  swing1Motion=Free
  swing2LimitDegrees=45
  swing2Motion=Free
  twistLimitDegrees=45
  twistMotion=Free
 }
 {bone1Name=dyn_skirt_A_02, bone2Name=dyn_skirt_A_01, swing1LimitDegrees=30, swing2LimitDegrees=15, twistLimitDegrees=1}
 {bone1Name=dyn_skirt_A_03, bone2Name=dyn_skirt_A_02, swing1LimitDegrees=20, swing2LimitDegrees=15, twistLimitDegrees=1}
 {bone1Name=dyn_skirt_B_01, bone2Name=base_skirt_B, swing1LimitDegrees=35, swing2LimitDegrees=20, twistLimitDegrees=0}
 {bone1Name=dyn_skirt_B_02, bone2Name=dyn_skirt_B_01, swing1LimitDegrees=35, swing2LimitDegrees=20, twistLimitDegrees=5}
 {bone1Name=dyn_skirt_B_03, bone2Name=dyn_skirt_B_02, swing1LimitDegrees=25, swing2LimitDegrees=20, twistLimitDegrees=0}
 {bone1Name=dyn_skirt_B_04, bone2Name=dyn_skirt_B_03, swing1LimitDegrees=15, swing2LimitDegrees=15, twistLimitDegrees=0}
 {bone1Name=dyn_skirt_D_01, bone2Name=base_skirt_D, swing1LimitDegrees=5, swing2LimitDegrees=1, twistLimitDegrees=1}
 {bone1Name=dyn_skirt_D_02, bone2Name=dyn_skirt_D_01, swing1LimitDegrees=10, swing2LimitDegrees=1, twistLimitDegrees=5}
 {bone1Name=dyn_skirt_D_03, bone2Name=dyn_skirt_D_02, swing1LimitDegrees=25, swing2LimitDegrees=1, twistLimitDegrees=0}
 {bone1Name=dyn_skirt_D_04, bone2Name=dyn_skirt_D_03, swing1LimitDegrees=45, swing2LimitDegrees=1, twistLimitDegrees=0}
 {bone1Name=dyn_skirt_E_01, bone2Name=base_skirt_E, swing1LimitDegrees=5, swing2LimitDegrees=10, twistLimitDegrees=0}
 {bone1Name=dyn_skirt_E_02, bone2Name=dyn_skirt_E_01, swing1LimitDegrees=10, swing2LimitDegrees=10, twistLimitDegrees=5}
 {bone1Name=dyn_skirt_E_03, bone2Name=dyn_skirt_E_02, swing1LimitDegrees=25, swing2LimitDegrees=5, twistLimitDegrees=0}
 {bone1Name=dyn_skirt_E_04, bone2Name=dyn_skirt_E_03, swing1LimitDegrees=45, swing2LimitDegrees=1, twistLimitDegrees=0}
 {bone1Name=dyn_skirt_F_01, bone2Name=base_skirt_F, swing1LimitDegrees=45, swing2LimitDegrees=10, twistLimitDegrees=0}
 {bone1Name=dyn_skirt_F_02, bone2Name=dyn_skirt_F_01, swing1LimitDegrees=35, swing2LimitDegrees=20, twistLimitDegrees=5}
 {bone1Name=dyn_skirt_F_03, bone2Name=dyn_skirt_F_02, swing1LimitDegrees=25, swing2LimitDegrees=20, twistLimitDegrees=0}
 {bone1Name=dyn_skirt_F_04, bone2Name=dyn_skirt_F_03, swing1LimitDegrees=15, swing2LimitDegrees=15, twistLimitDegrees=0}
 {bone1Name=dyn_skirt_G_01, bone2Name=base_skirt_G, swing1LimitDegrees=30, swing2LimitDegrees=15, twistLimitDegrees=0}
 {bone1Name=dyn_skirt_G_02, bone2Name=dyn_skirt_G_01, swing1LimitDegrees=30, swing2LimitDegrees=15, twistLimitDegrees=0}
 {bone1Name=dyn_skirt_G_03, bone2Name=dyn_skirt_G_02, swing1LimitDegrees=20, swing2LimitDegrees=15, twistLimitDegrees=5}
 {bone1Name=dyn_skirt_G_04, bone2Name=dyn_skirt_G_03, swing1LimitDegrees=15, swing2LimitDegrees=15, twistLimitDegrees=0}
 {bone1Name=dyn_skirt_C_01, bone2Name=base_skirt_C, swing1LimitDegrees=5, swing2LimitDegrees=10, twistLimitDegrees=0}
 {bone1Name=dyn_skirt_C_02, bone2Name=dyn_skirt_C_01, swing1LimitDegrees=10, swing2LimitDegrees=10, twistLimitDegrees=5}
 {bone1Name=dyn_skirt_C_03, bone2Name=dyn_skirt_C_02, swing1LimitDegrees=25, swing2LimitDegrees=5, twistLimitDegrees=0}
 {bone1Name=dyn_skirt_C_04, bone2Name=dyn_skirt_C_03, swing1LimitDegrees=45, swing2LimitDegrees=1, twistLimitDegrees=0}
 {bone1Name=dyn_skirt_A_01, bone2Name=base_skirt_A, swing1LimitDegrees=30, swing2LimitDegrees=15, twistLimitDegrees=1}
 {bone1Name=dyn_skirt_A_04, bone2Name=dyn_skirt_A_03, swing1LimitDegrees=15, swing2LimitDegrees=15, twistLimitDegrees=0}
 {
  bone1Name=base_skirt_A
  bone2Name=root
  swing1LimitDegrees=45
  swing1Motion=Free
  swing2LimitDegrees=45
  swing2Motion=Free
  twistLimitDegrees=45
  twistMotion=Free
 }
 {
  bone1Name=base_sleeve_l
  bone2Name=root
  swing1LimitDegrees=45
  swing1Motion=Free
  swing2LimitDegrees=45
  swing2Motion=Free
  twistLimitDegrees=45
  twistMotion=Free
 }
 {
  bone1Name=base_sleeve_r
  bone2Name=root
  swing1LimitDegrees=45
  swing1Motion=Free
  swing2LimitDegrees=45
  swing2Motion=Free
  twistLimitDegrees=45
  twistMotion=Free
 }
 {bone1Name=dyn_sleeve_l_1, bone2Name=base_sleeve_l, swing1LimitDegrees=20, swing2LimitDegrees=45, twistLimitDegrees=0}
 {bone1Name=dyn_sleeve_l_2, bone2Name=dyn_sleeve_l_1, swing1LimitDegrees=20, swing2LimitDegrees=45, twistLimitDegrees=0}
 {bone1Name=dyn_sleeve_l_3, bone2Name=dyn_sleeve_l_2, swing1LimitDegrees=20, swing2LimitDegrees=45, twistLimitDegrees=0}
 {bone1Name=dyn_sleeve_l_4, bone2Name=dyn_sleeve_l_3, swing1LimitDegrees=20, swing2LimitDegrees=45, twistLimitDegrees=0}
 {bone1Name=dyn_sleeve_r_1, bone2Name=base_sleeve_r, swing1LimitDegrees=20, swing2LimitDegrees=45, twistLimitDegrees=0}
 {bone1Name=dyn_sleeve_r_2, bone2Name=dyn_sleeve_r_1, swing1LimitDegrees=20, swing2LimitDegrees=45, twistLimitDegrees=0}
 {bone1Name=dyn_sleeve_r_3, bone2Name=dyn_sleeve_r_2, swing1LimitDegrees=20, swing2LimitDegrees=45, twistLimitDegrees=0}
]
