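# Animation Blueprint for the F_MED_PalmTree cosmetic body, driving sleeve and skirt secondary motion on top of the pose copied from the parent mesh.
# Two Control Rigs, Look At constraints and AnimDynamics feed cached RigidBody variants for gameplay, sliding and front end.
# Look At and offset fallbacks replace RigidBody when it is disabled during slide or crouch.
# The EventGraph turns movement, weapon and emote state into corrective-bone and physics alphas.
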AnimBlueprint /BRCosmetics/Characters/Player/Female/Medium/Bodies/F_MED_PalmTree/Meshes/F_MED_PalmTree_AnimBP
target_skeleton=&/F_MED_PalmTree_Skeleton
graphs [
 AnimGraph {nodes [
   Output Pose(Result=Use cached pose 'CRCache_BlendRB'()) {layerGroup=None}
   n1 = Control Rig(Source=Copy Pose From Mesh() {bUseAttachedParent=true}) {
    alphaBoolBlend {blendOption=Linear}
    controlRigAssetReference {blueprintRigClass=&/F_MED_PalmTree_ControlRig.F_MED_PalmTree_ControlRig_C}
   }
   n2 = Control Rig!(
    Source=@n1
    Alpha
    Crouching:double=Get Crouch() -> (Crouch:double)
    SkydiveMode:double=Get Skydiving() -> (Skydiving:double)
    TacticalJumping:double=Get TacticalJump() -> (TacticalJump:double)
    FrontEndAlpha:double=Get FrontEndAlpha() -> (FrontEndAlpha:double)
   ) -> (Pose) {
    alphaBoolBlend {blendOption=Linear}
    controlRigAssetReference {blueprintRigClass=&/F_MED_PalmTree_Sleeves_ControlRig.F_MED_PalmTree_Sleeves_ControlRig_C}
   }
   Save cached pose 'CRCache_Start'(Pose=@n2) {cacheName=CRCache_Start}
   n3 = Look At - Bone: base_sleeve_r(ComponentPose=Local To Component(LocalPose=Use cached pose 'CRCache_Start'()), Alpha=1) {
    alphaBoolBlend {blendOption=Linear}
    boneToModify {boneName=base_sleeve_r}
    lookAtLocation=<-4.1639 -7.3318 -1.1637>
    lookAtTarget {boneReference {boneName=calf_r}}
    lookAt_Axis {axis=<1 0>}
   }
   n4 = Look At - Bone: base_sleeve_l(ComponentPose=@n3, Alpha=1) {
    alphaBoolBlend {blendOption=Linear}
    boneToModify {boneName=base_sleeve_l}
    lookAtLocation=<12.7684 4.8992 4.8061>
    lookAtTarget {boneReference {boneName=calf_l}}
    lookAt_Axis {axis=<1 0>}
   }
   Save cached pose 'CRCache_constraints'(Pose=Component To Local(ComponentPose=@n4)) {cacheName=CRCache_constraints}
   n5 = "Transform (Modify) Bone - Bone: base_master_sleeve"!(
    ComponentPose=Local To Component(LocalPose=Use cached pose 'CRCache_constraints'())
    Alpha=Get jumping() -> (jumping:double)
    Translation="0,0,6"
   ) -> (Pose) {
    rotationMode=BMM_Additive
    translationMode=BMM_Additive
    alphaBoolBlend {blendOption=Linear}
    alphaScaleBiasClamp {bInterpResult=true}
    boneToModify {boneName=base_master_sleeve}
    translation {z=6}
   }
   n6 = "Transform (Modify) Bone - Bone: base_master_sleeve"(ComponentPose=@n5, Alpha=Get Crouch(), Translation="0,3.5,7.25") {
    rotationMode=BMM_Additive
    translationMode=BMM_Additive
    alphaBoolBlend {blendOption=Linear}
    alphaScaleBiasClamp {bInterpResult=true}
    boneToModify {boneName=base_master_sleeve}
    translation {y=3.5, z=7.25}
   }
   n7 = "Transform (Modify) Bone - Bone: base_master_sleeve"!(ComponentPose=@n6, Alpha=Get DBNOMode() -> (DBNOMode:double), Translation="13.5,4,0", Rotation:Rotator="-15,12.5,5") -> (Pose) {
    rotationMode=BMM_Additive
    rotationSpace=BCS_BoneSpace
    translationMode=BMM_Additive
    translationSpace=BCS_BoneSpace
    alphaBoolBlend {blendOption=Linear}
    alphaScaleBiasClamp {bInterpResult=true}
    boneToModify {boneName=base_master_sleeve}
    rotation=<rot -15 12.5 5>
    translation=<13.5 4>
   }
   Save cached pose 'CRCache_Adj'(Pose=Component To Local(ComponentPose=@n7)) {cacheName=CRCache_Adj}
   n8 = Get GravityOverride()
   n9 = Anim Dynamics - Bone: dyn_sleeve_A_l!(ComponentPose=Local To Component(LocalPose=Use cached pose 'CRCache_Adj'()), Alpha=1, GravityOverride:Vector=@n8, ExternalForce="0,0,0") -> (Pose) {
    angularDampingOverride=1
    angularSpringConstant=20000
    bAngularSpring=true
    bOverrideAngularDamping=true
    bUseGravityOverride=true
    gravityScale=0.35
    simulationSpace=RootRelative
    alphaBoolBlend {blendOption=Linear}
    boundBone {boneName=dyn_sleeve_A_l}
    componentAppliedLinearAccClamp=<3000 3000 3000>
    componentLinearAccScale=<0.3 0.4 0.4>
    componentLinearVelScale=<0.3 0.4 0.4>
    gravityOverride {z=-980}
    physicsBodyDefinitions [{
      boundBone {boneName=dyn_sleeve_A_l}
      boxExtents=<20 30 30>
      constraintSetup {angularLimitsMax=<7.5 0 7.5>, angularLimitsMin=<-7.5 0 -7.5>, angularTarget=<1 0 0>}
      localJointOffset=<1 0 0>
     }]
   }
   n10 = Anim Dynamics - Bone: dyn_sleeve_D_l!(ComponentPose=@n9, Alpha=1, GravityOverride:Vector=@n8, ExternalForce="0,0,0") -> (Pose) {
    angularDampingOverride=1
    angularSpringConstant=20000
    bAngularSpring=true
    bOverrideAngularDamping=true
    bUseGravityOverride=true
    gravityScale=0.35
    simulationSpace=RootRelative
    alphaBoolBlend {blendOption=Linear}
    boundBone {boneName=dyn_sleeve_D_l}
    componentAppliedLinearAccClamp=<3000 3000 3000>
    componentLinearAccScale=<0.3 0.4 0.4>
    componentLinearVelScale=<0.3 0.4 0.4>
    gravityOverride {z=-980}
    physicsBodyDefinitions [{
      boundBone {boneName=dyn_sleeve_D_l}
      boxExtents=<20 30 30>
      constraintSetup {angularLimitsMax=<7.5 0 7.5>, angularLimitsMin=<-7.5 0 -7.5>, angularTarget=<1 0 0>}
      localJointOffset=<1 0 0>
     }]
   }
   n11 = Anim Dynamics - Bone: dyn_sleeve_B_l!(ComponentPose=@n10, Alpha=1, GravityOverride:Vector=@n8, ExternalForce="0,0,0") -> (Pose) {
    angularDampingOverride=1
    angularSpringConstant=20000
    bAngularSpring=true
    bOverrideAngularDamping=true
    bUseGravityOverride=true
    gravityScale=0.35
    simulationSpace=RootRelative
    alphaBoolBlend {blendOption=Linear}
    boundBone {boneName=dyn_sleeve_B_l}
    componentAppliedLinearAccClamp=<3000 3000 3000>
    componentLinearAccScale=<0.3 0.4 0.4>
    componentLinearVelScale=<0.3 0.4 0.4>
    gravityOverride {z=-980}
    physicsBodyDefinitions [{
      boundBone {boneName=dyn_sleeve_B_l}
      boxExtents=<20 30 30>
      constraintSetup {angularLimitsMax=<7.5 0 7.5>, angularLimitsMin=<-7.5 0 -7.5>, angularTarget=<1 0 0>}
      localJointOffset=<1 0 0>
     }]
   }
   n12 = Anim Dynamics - Bone: dyn_sleeve_C_l!(ComponentPose=@n11, Alpha=1, GravityOverride:Vector=@n8, ExternalForce="0,0,0") -> (Pose) {
    angularDampingOverride=1
    angularSpringConstant=20000
    bAngularSpring=true
    bOverrideAngularDamping=true
    bUseGravityOverride=true
    gravityScale=0.35
    simulationSpace=RootRelative
    alphaBoolBlend {blendOption=Linear}
    boundBone {boneName=dyn_sleeve_C_l}
    componentAppliedLinearAccClamp=<3000 3000 3000>
    componentLinearAccScale=<0.3 0.4 0.4>
    componentLinearVelScale=<0.3 0.4 0.4>
    gravityOverride {z=-980}
    physicsBodyDefinitions [{
      boundBone {boneName=dyn_sleeve_C_l}
      boxExtents=<20 30 30>
      constraintSetup {angularLimitsMax=<7.5 0 7.5>, angularLimitsMin=<-7.5 0 -7.5>, angularTarget=<1 0 0>}
      localJointOffset=<1 0 0>
     }]
   }
   n13 = Anim Dynamics - Bone: dyn_sleeve_A_r!(ComponentPose=@n12, Alpha=1, GravityOverride:Vector=@n8, ExternalForce="0,0,0") -> (Pose) {
    angularDampingOverride=1
    angularSpringConstant=20000
    bAngularSpring=true
    bOverrideAngularDamping=true
    bUseGravityOverride=true
    gravityScale=0.35
    simulationSpace=RootRelative
    alphaBoolBlend {blendOption=Linear}
    boundBone {boneName=dyn_sleeve_A_r}
    componentAppliedLinearAccClamp=<3000 3000 3000>
    componentLinearAccScale=<0.3 0.4 0.4>
    componentLinearVelScale=<0.3 0.4 0.4>
    gravityOverride {z=-980}
    physicsBodyDefinitions [{
      boundBone {boneName=dyn_sleeve_A_r}
      boxExtents=<20 30 30>
      constraintSetup {angularLimitsMax=<7.5 0 7.5>, angularLimitsMin=<-7.5 0 -7.5>, angularTarget=<1 0 0>}
      localJointOffset=<1 0 0>
     }]
   }
   n14 = Anim Dynamics - Bone: dyn_sleeve_B_r!(ComponentPose=@n13, Alpha=1, GravityOverride:Vector=@n8, ExternalForce="0,0,0") -> (Pose) {
    angularDampingOverride=1
    angularSpringConstant=20000
    bAngularSpring=true
    bOverrideAngularDamping=true
    bUseGravityOverride=true
    gravityScale=0.35
    simulationSpace=RootRelative
    alphaBoolBlend {blendOption=Linear}
    boundBone {boneName=dyn_sleeve_B_r}
    componentAppliedLinearAccClamp=<3000 3000 3000>
    componentLinearAccScale=<0.3 0.4 0.4>
    componentLinearVelScale=<0.3 0.4 0.4>
    gravityOverride {z=-980}
    physicsBodyDefinitions [{
      boundBone {boneName=dyn_sleeve_B_r}
      boxExtents=<20 30 30>
      constraintSetup {angularLimitsMax=<7.5 0 7.5>, angularLimitsMin=<-7.5 0 -7.5>, angularTarget=<1 0 0>}
      localJointOffset=<1 0 0>
     }]
   }
   n15 = Anim Dynamics - Bone: dyn_sleeve_C_r!(ComponentPose=@n14, Alpha=1, GravityOverride:Vector=@n8, ExternalForce="0,0,0") -> (Pose) {
    angularDampingOverride=1
    angularSpringConstant=20000
    bAngularSpring=true
    bOverrideAngularDamping=true
    bUseGravityOverride=true
    gravityScale=0.35
    simulationSpace=RootRelative
    alphaBoolBlend {blendOption=Linear}
    boundBone {boneName=dyn_sleeve_C_r}
    componentAppliedLinearAccClamp=<3000 3000 3000>
    componentLinearAccScale=<0.3 0.4 0.4>
    componentLinearVelScale=<0.3 0.4 0.4>
    gravityOverride {z=-980}
    physicsBodyDefinitions [{
      boundBone {boneName=dyn_sleeve_C_r}
      boxExtents=<20 30 30>
      constraintSetup {angularLimitsMax=<7.5 0 7.5>, angularLimitsMin=<-7.5 0 -7.5>, angularTarget=<1 0 0>}
      localJointOffset=<1 0 0>
     }]
   }
   n16 = Anim Dynamics - Bone: dyn_sleeve_D_r!(ComponentPose=@n15, Alpha=1, GravityOverride:Vector=@n8, ExternalForce="0,0,0") -> (Pose) {
    angularDampingOverride=1
    angularSpringConstant=20000
    bAngularSpring=true
    bOverrideAngularDamping=true
    bUseGravityOverride=true
    gravityScale=0.35
    simulationSpace=RootRelative
    alphaBoolBlend {blendOption=Linear}
    boundBone {boneName=dyn_sleeve_D_r}
    componentAppliedLinearAccClamp=<3000 3000 3000>
    componentLinearAccScale=<0.3 0.4 0.4>
    componentLinearVelScale=<0.3 0.4 0.4>
    gravityOverride {z=-980}
    physicsBodyDefinitions [{
      boundBone {boneName=dyn_sleeve_D_r}
      boxExtents=<20 30 30>
      constraintSetup {angularLimitsMax=<7.5 0 7.5>, angularLimitsMin=<-7.5 0 -7.5>, angularTarget=<1 0 0>}
      localJointOffset=<1 0 0>
     }]
   }
   Save cached pose 'CRCache_Dyn'(Pose=Component To Local(ComponentPose=@n16)) {cacheName=CRCache_Dyn}
   n17 = Look At - Bone: dyn_skirt_A_01(ComponentPose=Local To Component(LocalPose=Use cached pose 'CRCache_Dyn'()), Alpha=1) {
    alphaBoolBlend {blendOption=Linear}
    boneToModify {boneName=dyn_skirt_A_01}
    lookAtLocation=<18.8604 -72.5615 -31.327>
    lookAtTarget {boneReference {boneName=deform_knee_out_l}}
    lookAt_Axis {axis=<1 0>}
   }
   n18 = Look At - Bone: base_skirt_B(ComponentPose=@n17, Alpha=1) {
    alphaBoolBlend {blendOption=Linear}
    boneToModify {boneName=base_skirt_B}
    lookAtLocation=<35.3104 -18.7822 0>
    lookAtTarget {boneReference {boneName=root}}
    lookAt_Axis {axis=<1 0>}
   }
   n19 = Look At - Bone: base_skirt_C(ComponentPose=@n18, Alpha=1) {
    alphaBoolBlend {blendOption=Linear}
    boneToModify {boneName=base_skirt_C}
    lookAtLocation=<24.7225 -19.6267 1.388>
    lookAtTarget {boneReference {boneName=root}}
    lookAt_Axis {axis=<1 0>}
   }
   n20 = Look At - Bone: base_skirt_D(ComponentPose=@n19, Alpha=1) {
    alphaBoolBlend {blendOption=Linear}
    boneToModify {boneName=base_skirt_D}
    lookAtLocation=<10.8313 -31.8592 0.9183>
    lookAtTarget {boneReference {boneName=root}}
    lookAt_Axis {axis=<1 0>}
   }
   n21 = Look At - Bone: base_skirt_E(ComponentPose=@n20, Alpha=1) {
    alphaBoolBlend {blendOption=Linear}
    boneToModify {boneName=base_skirt_E}
    lookAtLocation=<-5.4301 -52.6619 -0.7563>
    lookAtTarget {boneReference {boneName=root}}
    lookAt_Axis {axis=<1 0>}
   }
   n22 = Look At - Bone: base_skirt_F(ComponentPose=@n21, Alpha=1) {
    alphaBoolBlend {blendOption=Linear}
    boneToModify {boneName=base_skirt_F}
    lookAtLocation=<-19.2319 -44.2042 -0.9465>
    lookAtTarget {boneReference {boneName=root}}
    lookAt_Axis {axis=<1 0>}
   }
   n23 = Look At - Bone: dyn_skirt_G_01(ComponentPose=@n22, Alpha=1) {
    alphaBoolBlend {blendOption=Linear}
    boneToModify {boneName=dyn_skirt_G_01}
    lookAtLocation=<2.6686 14.6108 13.0641>
    lookAtTarget {boneReference {boneName=deform_knee_out_r}}
    lookAt_Axis {axis=<1 0>}
   }
   n24 = "Transform (Modify) Bone - Bone: dyn_skirt_E_02"!(ComponentPose=@n23, bAlphaBoolEnabled:bool=True, Rotation="0.024,4.616,0.004") -> (Pose) {
    alphaInputType=Bool
    rotationMode=BMM_Additive
    rotationSpace=BCS_BoneSpace
    translationMode=BMM_Additive
    translationSpace=BCS_BoneSpace
    alphaBoolBlend {blendOption=Linear}
    boneToModify {boneName=dyn_skirt_E_02}
    rotation=<rot 0.024 4.616 0.004>
   }
   n25 = Get crouchMove() -> (crouchMove:double)
   n26 = "Transform (Modify) Bone - Bone: dyn_skirt_G_02"!(ComponentPose=@n24, Alpha=@n25, Rotation="0,-25,0") -> (Pose) {
    rotationMode=BMM_Additive
    rotationSpace=BCS_BoneSpace
    translationMode=BMM_Additive
    translationSpace=BCS_BoneSpace
    alphaBoolBlend {blendOption=Linear}
    alphaScaleBiasClamp {bInterpResult=true}
    boneToModify {boneName=dyn_skirt_G_02}
    rotation {yaw=-25}
   }
   n27 = "Transform (Modify) Bone - Bone: base_sleeve_l"!(ComponentPose=@n26, Alpha=@n25, Translation="1,0,0", Rotation="0,10,-25") -> (Pose) {
    rotationMode=BMM_Additive
    translationMode=BMM_Additive
    alphaBoolBlend {blendOption=Linear}
    alphaScaleBiasClamp {bInterpResult=true}
    boneToModify {boneName=base_sleeve_l}
    rotation {roll=-25, yaw=10}
    translation {x=1}
   }
   n28 = "Transform (Modify) Bone - Bone: base_sleeve_l"!(ComponentPose=@n27, bAlphaBoolEnabled:bool=True, Translation="0,0,2.5", Rotation="-0.749,2.595,32.145") -> (Pose) {
    alphaInputType=Bool
    rotationMode=BMM_Additive
    translationMode=BMM_Additive
    alphaBoolBlend {blendOption=Linear}
    alphaScaleBiasClamp {interpSpeedDecreasing=5, interpSpeedIncreasing=5}
    boneToModify {boneName=base_sleeve_l}
    rotation=<rot -0.749 2.595 32.145>
    translation {z=2.5}
   }
   Save cached pose 'CRCache_Crouch_RBoff'(Pose=Component To Local(ComponentPose=@n28)) {cacheName=CRCache_Crouch_RBoff}
   n29 = Look At - Bone: dyn_skirt_A_01(ComponentPose=Local To Component(LocalPose=Use cached pose 'CRCache_Dyn'()), Alpha=1) {
    alphaBoolBlend {blendOption=Linear}
    boneToModify {boneName=dyn_skirt_A_01}
    lookAtLocation=<63.2114 19.085 1.9774>
    lookAtTarget {boneReference {boneName=root}}
    lookAt_Axis {axis=<1 0>}
   }
   n30 = Look At - Bone: base_skirt_B(ComponentPose=@n29, Alpha=1) {
    alphaBoolBlend {blendOption=Linear}
    boneToModify {boneName=base_skirt_B}
    lookAtLocation=<63.1667 -33.5361 8.194>
    lookAtTarget {boneReference {boneName=root}}
    lookAt_Axis {axis=<1 0>}
   }
   n31 = Look At - Bone: base_skirt_C(ComponentPose=@n30, Alpha=1) {
    alphaBoolBlend {blendOption=Linear}
    boneToModify {boneName=base_skirt_C}
    lookAtLocation=<40.8918 -56.769 0.4962>
    lookAtTarget {boneReference {boneName=root}}
    lookAt_Axis {axis=<1 0>}
   }
   n32 = Look At - Bone: base_skirt_D(ComponentPose=@n31, Alpha=1) {
    alphaBoolBlend {blendOption=Linear}
    boneToModify {boneName=base_skirt_D}
    lookAtLocation=<0 -77.3673 12.239>
    lookAtTarget {boneReference {boneName=root}}
    lookAt_Axis {axis=<1 0>}
   }
   n33 = Look At - Bone: base_skirt_E(ComponentPose=@n32, Alpha=1) {
    alphaBoolBlend {blendOption=Linear}
    boneToModify {boneName=base_skirt_E}
    lookAtLocation=<-46.4407 -67.7825 3.7389>
    lookAtTarget {boneReference {boneName=root}}
    lookAt_Axis {axis=<1 0>}
   }
   n34 = Look At - Bone: base_skirt_F(ComponentPose=@n33, Alpha=1) {
    alphaBoolBlend {blendOption=Linear}
    boneToModify {boneName=base_skirt_F}
    lookAtLocation=<-69.1034 -24.0776 -13.3816>
    lookAtTarget {boneReference {boneName=root}}
    lookAt_Axis {axis=<1 0>}
   }
   n35 = Look At - Bone: base_skirt_G(ComponentPose=@n34, Alpha=1) {
    alphaBoolBlend {blendOption=Linear}
    boneToModify {boneName=base_skirt_G}
    lookAtLocation=<-75.843 1.9279 5.5388>
    lookAtTarget {boneReference {boneName=root}}
    lookAt_Axis {axis=<1 0>}
   }
   n36 = "Transform (Modify) Bone - Bone: base_skirt_C"!(ComponentPose=@n35, bAlphaBoolEnabled:bool=True, Translation="0,0,7") -> (Pose) {
    alphaInputType=Bool
    rotationMode=BMM_Additive
    translationMode=BMM_Additive
    alphaBoolBlend {blendOption=Linear}
    boneToModify {boneName=base_skirt_C}
    translation {z=7}
   }
   n37 = "Transform (Modify) Bone - Bone: base_skirt_D"!(ComponentPose=@n36, bAlphaBoolEnabled:bool=True, Translation="0,0,7.5") -> (Pose) {
    alphaInputType=Bool
    rotationMode=BMM_Additive
    translationMode=BMM_Additive
    alphaBoolBlend {blendOption=Linear}
    boneToModify {boneName=base_skirt_D}
    translation {z=7.5}
   }
   n38 = "Transform (Modify) Bone - Bone: base_skirt_E"!(ComponentPose=@n37, bAlphaBoolEnabled:bool=True, Translation="0,0,7.75") -> (Pose) {
    alphaInputType=Bool
    rotationMode=BMM_Additive
    translationMode=BMM_Additive
    alphaBoolBlend {blendOption=Linear}
    boneToModify {boneName=base_skirt_E}
    translation {z=7.75}
   }
   n39 = "Transform (Modify) Bone - Bone: base_skirt_F"!(ComponentPose=@n38, bAlphaBoolEnabled:bool=True, Translation="0,0,7") -> (Pose) {
    alphaInputType=Bool
    rotationMode=BMM_Additive
    translationMode=BMM_Additive
    alphaBoolBlend {blendOption=Linear}
    boneToModify {boneName=base_skirt_F}
    translation {z=7}
   }
   n40 = "Transform (Modify) Bone - Bone: base_master_sleeve"!(ComponentPose=@n39, bAlphaBoolEnabled:bool=True, Translation="0,0,3", Rotation="0,0,0") -> (Pose) {
    alphaInputType=Bool
    rotationMode=BMM_Additive
    translationMode=BMM_Additive
    alphaBoolBlend {blendOption=Linear}
    boneToModify {boneName=base_master_sleeve}
    translation {z=3}
   }
   n41 = RigidBody!(ComponentPose=@n40, Alpha=0.5, OverrideWorldGravity:Vector="0,0,-980") -> (Pose) {
    bClampLinearTranslationLimitToRefPose=true
    bOverrideWorldGravity=true
    overridePhysicsAsset=&/F_MED_PalmTree_Sliding_Physics
    simulationSpace=BaseBoneSpace
    alphaBoolBlend {blendOption=Linear}
    baseBoneRef {boneName=pelvis}
    componentAppliedLinearAccClamp=<3000 3000 3000>
    componentLinearAccScale=<0.3 0.2 0.2>
    componentLinearVelScale=<0.3 0.2 0.2>
    overrideWorldGravity {z=-980}
    simSpaceSettings {
     maxAngularAcceleration=1002
     maxAngularVelocity=3
     maxLinearAcceleration=700
     maxLinearVelocity=500
     velocityScaleZ=0
     worldAlpha=0.5
     externalLinearDragV=<0.2 0.2>
    }
   }
   Save cached pose 'CRCache_Slide_RB'(Pose=Component To Local(ComponentPose=@n41)) {cacheName=CRCache_Slide_RB}
   n42 = RigidBody(
    ComponentPose=Local To Component(LocalPose=Use cached pose 'CRCache_Dyn'())
    Alpha=Get SkirtDynAlpha() -> (SkirtDynAlpha:double)
    OverrideWorldGravity="0,0,-980"
   ) {
    bClampLinearTranslationLimitToRefPose=true
    bOverrideWorldGravity=true
    overridePhysicsAsset=&/F_MED_PalmTree_Physics
    simulationSpace=BaseBoneSpace
    alphaBoolBlend {blendOption=Linear}
    alphaScaleBiasClamp {bInterpResult=true}
    baseBoneRef {boneName=pelvis}
    componentAppliedLinearAccClamp=<3000 3000 3000>
    componentLinearAccScale=<0.2 0.1 0.1>
    componentLinearVelScale=<0.2 0.1 0.1>
    overrideWorldGravity {z=-980}
    simSpaceSettings {
     maxAngularAcceleration=1002
     maxAngularVelocity=3
     maxLinearAcceleration=700
     maxLinearVelocity=500
     velocityScaleZ=0
     worldAlpha=0.5
     externalLinearDragV=<0.2 0.2>
    }
   }
   n43 = "Transform (Modify) Bone - Bone: base_sleeve_l"!(ComponentPose=@n42, Alpha:float=1, Translation="0,2.5,1") -> (Pose) {
    rotationMode=BMM_Additive
    rotationSpace=BCS_BoneSpace
    translationMode=BMM_Additive
    alphaBoolBlend {blendOption=Linear}
    boneToModify {boneName=base_sleeve_l}
    translation {y=2.5, z=1}
   }
   n44 = RigidBody(ComponentPose=@n43, Alpha=Get SleevesDynAlpha() -> (SleevesDynAlpha:double), OverrideWorldGravity="0,0,-980") {
    bClampLinearTranslationLimitToRefPose=true
    bOverrideWorldGravity=true
    overridePhysicsAsset=&/F_MED_PalmTree_BackJacket_Physics
    simulationSpace=BaseBoneSpace
    alphaBoolBlend {blendOption=Linear}
    alphaScaleBiasClamp {bInterpResult=true}
    baseBoneRef {boneName=root}
    componentAppliedLinearAccClamp=<3000 3000 3000>
    componentLinearAccScale=<0.2 0.1 0.1>
    componentLinearVelScale=<0.2 0.1 0.1>
    overrideWorldGravity {z=-980}
    simSpaceSettings {
     maxAngularAcceleration=1002
     maxAngularVelocity=3
     maxLinearAcceleration=700
     maxLinearVelocity=500
     velocityScaleZ=0
     worldAlpha=0.5
     externalLinearDragV=<0.2 0.2>
    }
   }
   n45 = Get Crouch()
   n46 = "Transform (Modify) Bone - Bone: dyn_skirt_A_02"!(ComponentPose=@n44, Alpha=@n45, Translation="-2.5,0,-12.5", Rotation="0,20,15") -> (Pose) {
    rotationMode=BMM_Additive
    translationMode=BMM_Additive
    alphaBoolBlend {blendOption=Linear}
    alphaScaleBiasClamp {bInterpResult=true}
    boneToModify {boneName=dyn_skirt_A_02}
    rotation {roll=15, yaw=20}
    translation {x=-2.5, z=-12.5}
   }
   n47 = "Transform (Modify) Bone - Bone: base_skirt_E"!(ComponentPose=@n46, Alpha:float=@n45, Rotation:Rotator="-10,0,10") -> (Pose) {
    rotationMode=BMM_Additive
    translationMode=BMM_Additive
    alphaBoolBlend {blendOption=Linear}
    alphaScaleBiasClamp {bInterpResult=true}
    boneToModify {boneName=base_skirt_E}
    rotation {pitch=-10, roll=10}
   }
   n48 = Get crouchMove()
   n49 = "Transform (Modify) Bone - Bone: dyn_skirt_A_01"!(ComponentPose=@n47, Alpha=@n48, Translation="0,0,1.5", Rotation="10,0,0") -> (Pose) {
    rotationMode=BMM_Additive
    translationMode=BMM_Additive
    alphaBoolBlend {blendOption=Linear}
    alphaScaleBiasClamp {bInterpResult=true}
    boneToModify {boneName=dyn_skirt_A_01}
    rotation {pitch=10}
    translation {z=1.5}
   }
   n50 = "Transform (Modify) Bone - Bone: base_skirt_G"!(ComponentPose=@n49, Alpha=@n48, Translation="-2.5,0,2.5") -> (Pose) {
    rotationMode=BMM_Additive
    translationMode=BMM_Additive
    alphaBoolBlend {blendOption=Linear}
    alphaScaleBiasClamp {bInterpResult=true}
    boneToModify {boneName=base_skirt_G}
    translation {x=-2.5, z=2.5}
   }
   n51 = "Transform (Modify) Bone - Bone: dyn_skirt_G_01"!(ComponentPose=@n50, Alpha=Get Running() -> (Running:double), Translation="-2.5,-2.5,0") -> (Pose) {
    rotationMode=BMM_Additive
    translationMode=BMM_Additive
    alphaBoolBlend {blendOption=Linear}
    alphaScaleBiasClamp {bInterpResult=true}
    boneToModify {boneName=dyn_skirt_G_01}
    translation=<-2.5 -2.5>
   }
   n52 = Get Running()
   n53 = "Transform (Modify) Bone - Bone: base_sleeve_l"(ComponentPose=@n51, Alpha=@n52, Translation="1,0,4") {
    rotationMode=BMM_Additive
    rotationSpace=BCS_BoneSpace
    translationMode=BMM_Additive
    alphaBoolBlend {blendOption=Linear}
    alphaScaleBiasClamp {bInterpResult=true}
    boneToModify {boneName=base_sleeve_l}
    translation {x=1, z=4}
   }
   n54 = "Transform (Modify) Bone - Bone: base_sleeve_r"!(ComponentPose=@n53, Alpha=@n52, Rotation="-7.5,-0.7,-5") -> (Pose) {
    rotationMode=BMM_Additive
    translationMode=BMM_Additive
    translationSpace=BCS_BoneSpace
    alphaBoolBlend {blendOption=Linear}
    alphaScaleBiasClamp {bInterpResult=true, interpSpeedDecreasing=5, interpSpeedIncreasing=5}
    boneToModify {boneName=base_sleeve_r}
    rotation=<rot -7.5 -0.7 -5>
   }
   n55 = "Transform (Modify) Bone - Bone: base_sleeve_l"!(ComponentPose=@n54, Alpha=Get crouchMove(), Rotation:Rotator="-5,0,0") -> (Pose) {
    rotationMode=BMM_Additive
    rotationSpace=BCS_BoneSpace
    scaleSpace=BCS_BoneSpace
    translationMode=BMM_Additive
    translationSpace=BCS_BoneSpace
    alphaBoolBlend {blendOption=Linear}
    alphaScaleBiasClamp {bInterpResult=true, interpSpeedDecreasing=7.5, interpSpeedIncreasing=7.5}
    boneToModify {boneName=base_sleeve_l}
    rotation {pitch=-5}
   }
   n56 = "Transform (Modify) Bone - Bone: base_sleeve_l"!(ComponentPose=@n55, Alpha=Get Skydiving(), Translation:Vector="0,0,-8.5", Rotation="0,0,0") -> (Pose) {
    rotationMode=BMM_Additive
    translationMode=BMM_Additive
    alphaBoolBlend {blendOption=Linear}
    alphaScaleBiasClamp {bInterpResult=true, interpSpeedDecreasing=5, interpSpeedIncreasing=5}
    boneToModify {boneName=base_sleeve_l}
    translation {z=-8.5}
   }
   n57 = "Transform (Modify) Bone - Bone: base_sleeve_l"(ComponentPose=@n56, Alpha=Get crouchSleeve() -> (crouchSleeve:double), Translation="-3,-3,6.5", Rotation="-30,20,20") {
    rotationMode=BMM_Additive
    rotationSpace=BCS_BoneSpace
    translationMode=BMM_Additive
    translationSpace=BCS_BoneSpace
    alphaBoolBlend {blendOption=Linear}
    alphaScaleBiasClamp {bInterpResult=true, interpSpeedDecreasing=2.5}
    boneToModify {boneName=base_sleeve_l}
    rotation=<rot -30 20 20>
    translation=<-3 -3 6.5>
   }
   n58 = "Transform (Modify) Bone - Bone: base_sleeve_l"(ComponentPose=@n57, Alpha=Get CrouchGun() -> (CrouchGun:double), Translation="-1.75,-1.25,3.328", Rotation="-15,40,5") {
    rotationMode=BMM_Additive
    rotationSpace=BCS_BoneSpace
    translationMode=BMM_Additive
    translationSpace=BCS_BoneSpace
    alphaBoolBlend {blendOption=Linear}
    alphaScaleBiasClamp {bInterpResult=true, interpSpeedDecreasing=2.5, interpSpeedIncreasing=5}
    boneToModify {boneName=base_sleeve_l}
    rotation=<rot -15 40 5>
    translation=<-1.75 -1.25 3.328>
   }
   Save cached pose 'CRCache_Base_RB'(Pose=Component To Local(ComponentPose=@n58)) {cacheName=CRCache_Base_RB}
   n59 = Look At - Bone: dyn_skirt_A_01(ComponentPose=Local To Component(LocalPose=Use cached pose 'CRCache_Dyn'()), Alpha=1) {
    alphaBoolBlend {blendOption=Linear}
    boneToModify {boneName=dyn_skirt_A_01}
    lookAtLocation=<63.2113 19.085 -2.9936>
    lookAtTarget {boneReference {boneName=root}}
    lookAt_Axis {axis=<1 0>}
   }
   n60 = Look At - Bone: base_skirt_B(ComponentPose=@n59, Alpha=1) {
    alphaBoolBlend {blendOption=Linear}
    boneToModify {boneName=base_skirt_B}
    lookAtLocation=<63.1667 -33.5361 6.9088>
    lookAtTarget {boneReference {boneName=root}}
    lookAt_Axis {axis=<1 0>}
   }
   n61 = Look At - Bone: base_skirt_C(ComponentPose=@n60, Alpha=1) {
    alphaBoolBlend {blendOption=Linear}
    boneToModify {boneName=base_skirt_C}
    lookAtLocation=<40.8918 -56.769 0.8207>
    lookAtTarget {boneReference {boneName=root}}
    lookAt_Axis {axis=<1 0>}
   }
   n62 = Look At - Bone: base_skirt_D(ComponentPose=@n61, Alpha=1) {
    alphaBoolBlend {blendOption=Linear}
    boneToModify {boneName=base_skirt_D}
    lookAtLocation=<0 -77.3673 6.9612>
    lookAtTarget {boneReference {boneName=root}}
    lookAt_Axis {axis=<1 0>}
   }
   n63 = Look At - Bone: base_skirt_E(ComponentPose=@n62, Alpha=1) {
    alphaBoolBlend {blendOption=Linear}
    boneToModify {boneName=base_skirt_E}
    lookAtLocation=<-46.4407 -67.7825 -5.8671>
    lookAtTarget {boneReference {boneName=root}}
    lookAt_Axis {axis=<1 0>}
   }
   n64 = Look At - Bone: base_skirt_F(ComponentPose=@n63, Alpha=1) {
    alphaBoolBlend {blendOption=Linear}
    boneToModify {boneName=base_skirt_F}
    lookAtLocation=<-69.1034 -24.0776 -19.3385>
    lookAtTarget {boneReference {boneName=root}}
    lookAt_Axis {axis=<1 0>}
   }
   n65 = Look At - Bone: base_skirt_G(ComponentPose=@n64, Alpha=1) {
    alphaBoolBlend {blendOption=Linear}
    boneToModify {boneName=base_skirt_G}
    lookAtLocation=<-75.843 1.9279 5.5388>
    lookAtTarget {boneReference {boneName=root}}
    lookAt_Axis {axis=<1 0>}
   }
   n66 = "Transform (Modify) Bone - Bone: base_skirt_C"(ComponentPose=@n65, bAlphaBoolEnabled=True, Translation="0,0,6.5") {
    alphaInputType=Bool
    rotationMode=BMM_Additive
    translationMode=BMM_Additive
    alphaBoolBlend {blendOption=Linear}
    boneToModify {boneName=base_skirt_C}
    translation {z=6.5}
   }
   n67 = "Transform (Modify) Bone - Bone: base_skirt_D"(ComponentPose=@n66, bAlphaBoolEnabled=True, Translation="0,0,6.5") {
    alphaInputType=Bool
    rotationMode=BMM_Additive
    translationMode=BMM_Additive
    alphaBoolBlend {blendOption=Linear}
    boneToModify {boneName=base_skirt_D}
    translation {z=6.5}
   }
   n68 = "Transform (Modify) Bone - Bone: base_skirt_E"!(ComponentPose=@n67, bAlphaBoolEnabled:bool=True, Translation:Vector="0,0,6.5") -> (Pose) {
    alphaInputType=Bool
    rotationMode=BMM_Additive
    translationMode=BMM_Additive
    alphaBoolBlend {blendOption=Linear}
    boneToModify {boneName=base_skirt_E}
    translation {z=6.5}
   }
   n69 = "Transform (Modify) Bone - Bone: base_skirt_F"(ComponentPose=@n68, bAlphaBoolEnabled=True, Translation="0,0,6.5") {
    alphaInputType=Bool
    rotationMode=BMM_Additive
    translationMode=BMM_Additive
    alphaBoolBlend {blendOption=Linear}
    boneToModify {boneName=base_skirt_F}
    translation {z=6.5}
   }
   n70 = "Transform (Modify) Bone - Bone: base_master_sleeve"(ComponentPose=@n69, bAlphaBoolEnabled=True, Translation="0,0,3", Rotation="0,0,0") {
    alphaInputType=Bool
    rotationMode=BMM_Additive
    translationMode=BMM_Additive
    alphaBoolBlend {blendOption=Linear}
    boneToModify {boneName=base_master_sleeve}
    translation {z=3}
   }
   Save cached pose 'CRCache_Slide_RBoff'(Pose=Component To Local(ComponentPose=@n70)) {cacheName=CRCache_Slide_RBoff}
   n71 = Blend Poses by bool(
    BlendPose_0=Use cached pose 'CRCache_Slide_RB'()
    BlendPose_1=Blend Poses by bool(
     BlendPose_0=Use cached pose 'CRCache_FE'()
     BlendPose_1=Use cached pose 'CRCache_Base_RB'()
     BlendTime_0=0.5
     bActiveValue=Get bIsInFrontEnd()
    ) {blendType=Linear}
    BlendTime_0=0.5
    bActiveValue=Get bIsSliding()
   ) {blendType=Linear}
   n72 = Use cached pose 'CRCache_Crouch_RBoff'()
   n73 = Get CrouchRBdisabled() -> (CrouchRBdisabled:bool)
   Save cached pose 'CRCache_BlendRB'(Pose=Blend Poses by bool(
     BlendPose_0=Use cached pose 'CRCache_Slide_RBoff'()
     BlendPose_1=Blend Poses by bool(BlendPose_0=@n72, BlendPose_1=@n71, bActiveValue=@n73) {blendType=Linear}
     bActiveValue=Get SlidingRBdisabled() -> (SlidingRBdisabled:bool)
    ) {blendType=Linear}) {cacheName=CRCache_BlendRB}
   n74 = RigidBody(
    ComponentPose=Local To Component(LocalPose=Use cached pose 'CRCache_Dyn'())
    Alpha=Get SkirtDynAlpha_FE() -> (SkirtDynAlpha_FE:double)
    OverrideWorldGravity="0,0,-980"
   ) {
    bClampLinearTranslationLimitToRefPose=true
    bOverrideWorldGravity=true
    overridePhysicsAsset=&F_MED_PalmTree_Physics
    simulationSpace=BaseBoneSpace
    alphaBoolBlend {blendOption=Linear}
    alphaScaleBiasClamp {bInterpResult=true}
    baseBoneRef {boneName=pelvis}
    componentAppliedLinearAccClamp=<3000 3000 3000>
    componentLinearAccScale=<0.2 0.1 0.1>
    componentLinearVelScale=<0.2 0.1 0.1>
    overrideWorldGravity {z=-980}
    simSpaceSettings {
     maxAngularAcceleration=1002
     maxAngularVelocity=3
     maxLinearAcceleration=700
     maxLinearVelocity=500
     velocityScaleZ=0
     worldAlpha=0.5
     externalLinearDragV=<0.2 0.2>
    }
   }
   n75 = "Transform (Modify) Bone - Bone: base_sleeve_l"(ComponentPose=@n74, Alpha=1, Translation="0,1,1", Rotation="10,0,0") {
    rotationMode=BMM_Additive
    rotationSpace=BCS_BoneSpace
    translationMode=BMM_Additive
    alphaBoolBlend {blendOption=Linear}
    boneToModify {boneName=base_sleeve_l}
    rotation {pitch=10}
    translation {y=1, z=1}
   }
   n76 = RigidBody(ComponentPose=@n75, Alpha=Get SleevesDynAlpha_FE() -> (SleevesDynAlpha_FE:double), OverrideWorldGravity="0,0,-980") {
    bClampLinearTranslationLimitToRefPose=true
    bOverrideWorldGravity=true
    overridePhysicsAsset=&F_MED_PalmTree_BackJacket_Physics
    simulationSpace=BaseBoneSpace
    alphaBoolBlend {blendOption=Linear}
    baseBoneRef {boneName=root}
    componentAppliedLinearAccClamp=<3000 3000 3000>
    componentLinearAccScale=<0.2 0.1 0.1>
    componentLinearVelScale=<0.2 0.1 0.1>
    overrideWorldGravity {z=-980}
    simSpaceSettings {
     maxAngularAcceleration=1002
     maxAngularVelocity=3
     maxLinearAcceleration=700
     maxLinearVelocity=500
     velocityScaleZ=0
     worldAlpha=0.5
     externalLinearDragV=<0.2 0.2>
    }
   }
   n77 = Get FE_Idle() -> (FE_Idle:double)
   n78 = "Transform (Modify) Bone - Bone: base_skirt_B"!(ComponentPose=@n76, Alpha=@n77, Translation="0,-2,0", Rotation="-5,-5,0") -> (Pose) {
    rotationMode=BMM_Additive
    rotationSpace=BCS_BoneSpace
    translationMode=BMM_Additive
    translationSpace=BCS_BoneSpace
    alphaBoolBlend {blendOption=Linear}
    alphaScaleBiasClamp {bInterpResult=true}
    boneToModify {boneName=base_skirt_B}
    rotation {pitch=-5, yaw=-5}
    translation {y=-2}
   }
   n79 = "Transform (Modify) Bone - Bone: base_skirt_F"!(ComponentPose=@n78, Alpha:float=@n77, Translation="0,0,0", Rotation:Rotator="5,5,0") -> (Pose) {
    rotationMode=BMM_Additive
    rotationSpace=BCS_BoneSpace
    translationMode=BMM_Additive
    translationSpace=BCS_BoneSpace
    alphaBoolBlend {blendOption=Linear}
    alphaScaleBiasClamp {bInterpResult=true}
    boneToModify {boneName=base_skirt_F}
    rotation {pitch=5, yaw=5}
   }
   n80 = "Transform (Modify) Bone - Bone: base_sleeve_l"(ComponentPose=@n79, Alpha=Get Emoting() -> (Emoting:double), Translation="0,2,2", Rotation="-5,0,0") {
    rotationMode=BMM_Additive
    rotationSpace=BCS_BoneSpace
    translationMode=BMM_Additive
    alphaBoolBlend {blendOption=Linear}
    alphaScaleBiasClamp {bInterpResult=true}
    boneToModify {boneName=base_sleeve_l}
    rotation {pitch=-5}
    translation {y=2, z=2}
   }
   Save cached pose 'CRCache_FE'(Pose=Component To Local(ComponentPose=@n80)) {cacheName=CRCache_FE}
  ]}
 EventGraph {nodes [
   Event BlueprintUpdateAnimation()
   > n1 = Sequence!(execute) -> (then_0, then_1, then_2:Exec, then_3:Exec)
   n2 = Sequence!(execute=@n1.then_1) -> (then_0, then_1, then_2)
   n3 = Get bIsJumping()
   n4 = Get bIsFalling()
   Set jumping(execute=@n2.then_1, jumping:double="To Float (Boolean)"(InBool=OR Boolean(A=@n3, B=@n4))) -> (Output_Get:double)
   > Set Crouch(Crouch:double="To Float (Boolean)"(InBool=Get bIsCrouching())) -> (Output_Get:double)
   n5 = Get bIsSprinting()
   n6 = Get bIsMoving2D()
   > Set Running(Running:double="To Float (Boolean)"(InBool=OR Boolean(A=@n6, B=@n5))) -> (Output_Get:double)
   n7 = OR Boolean(A=Get bIsJumping(), B=Get bIsFalling())
   n8 = Get bIsTacticalSprinting()
   > Set TacticalJump(TacticalJump:double="To Float (Boolean)"(InBool=AND Boolean(A=@n8, B=@n7))) -> (Output_Get:double)
   n9 = NOT Boolean(A=Get bIsMoving2D())
   n10 = Get bIsCrouching()
   n11 = Get Gun_out() -> (Gun_out:bool)
   > Set CrouchGun(CrouchGun:double="To Float (Boolean)"(InBool=AND Boolean!(A=@n11, B=@n10, C:bool=@n9) -> (ReturnValue))) -> (Output_Get:double)
   > Set Emoting(Emoting:double="To Float (Boolean)"(InBool=Get bIsPlayingEmote())) -> (Output_Get:double)
   > Set DBNOMode(DBNOMode:double="To Float (Boolean)"(InBool=Get bIsDBNO())) -> (Output_Get:double)
   n12 = Get bIsSliding()
   Set Sliding(execute=@n2.then_0, Sliding:double="To Float (Boolean)"(InBool=@n12)) -> (Output_Get:double)
   n13 = Get bRigidBodyEnabled()
   > Set SlidingRBdisabled(SlidingRBdisabled:bool=AND Boolean!(A=@n12, B=NOT Boolean(A=@n13)) -> (ReturnValue)) -> (Output_Get:bool)
   n14 = Get bRigidBodyEnabled()
   > Set CrouchRBdisabled(CrouchRBdisabled:bool=AND Boolean(A=Get bIsCrouching(), B=NOT Boolean(A=@n14))) -> (Output_Get:bool)
   n15 = Get UpperArmRot_l() -> (UpperArmRot_l:bool)
   n16 = Get bIsTargeting()
   > Set targetUpAlpha(targetUpAlpha:double="To Float (Boolean)"(InBool=AND Boolean(A=@n15, B=@n16))) -> (Output_Get:double)
   n17 = Get bIsMoving2D()
   Set moving2D(execute=@n2.then_2, moving2D:double="To Float (Boolean)"(InBool=@n17)) -> (Output_Get:double)
   n18 = Get bIsCrouching()
   > Set crouchMove(crouchMove:double="To Float (Boolean)"(InBool=AND Boolean(A=@n17, B=@n18))) -> (Output_Get:double)
   n19 = Get bIsSkydiveFloating()
   n20 = Get bIsSkydiveDiveMode()
   n21 = Get bIsParachuteOpened()
   n22 = Get bIsSkyDiving()
   > Set Skydiving(Skydiving:double="To Float (Boolean)"(InBool=OR Boolean!(A=@n19, B=@n20, C:bool=@n21, D:bool=@n22) -> (ReturnValue))) -> (Output_Get:double)
   n23 = AND Boolean(A=Get bIsCrouching(), B=NOT Boolean(A=Get bIsMoving2D()))
   n24 = Get bIsOnGrindRail()
   > Set crouchSleeve(crouchSleeve:double="To Float (Boolean)"(InBool=OR Boolean!(A=@n24, B=@n23) -> (ReturnValue))) -> (Output_Get:double)
   n25 = NOT Boolean(A=Get bIsPlayingEmote())
   n26 = Get bIsInFrontEnd()
   > Set FE_Idle(FE_Idle:double="To Float (Boolean)"(InBool=AND Boolean(A=@n26, B=@n25))) -> (Output_Get:double)
   n27 = Sequence(execute=@n1.then_2)
   n28 = Get bIsSkydiveFloating()
   n29 = Get bIsSkydiveDiveMode()
   n30 = Get bIsParachuteOpened()
   n31 = Get bIsSkyDiving()
   n32 = Get bIsSprinting()
   n33 = Get bIsMoving2D()
   n34 = Get bIsCrouching()
   n35 = Get bIsMoving2D()
   n36 = Branch(
    execute=@n27.then_1
    Condition=OR Boolean!(
     A=AND Boolean(A=@n35, B=@n34)
     B=OR Boolean(A=@n28, B=@n29, C=@n30, D=@n31)
     C:bool=Get bIsOnGrindRail()
     D:bool=OR Boolean!(A=@n33, B=@n32) -> (ReturnValue)
    ) -> (ReturnValue)
   )
   > Set SleevesDynAlpha(SleevesDynAlpha:double=0.45) -> (Output_Get:double)
   Set SleevesDynAlpha(execute=@n36.else, SleevesDynAlpha=1)
   n37 = Get bIsCrouching()
   n38 = Get Gun_out()
   > n39 = Branch(Condition=OR Boolean(A=Get bIsDBNO(), B=AND Boolean(A=@n37, B=@n38)))
   > Set SleevesDynAlpha(SleevesDynAlpha=0)
   Set SleevesDynAlpha(execute=@n39.else, SleevesDynAlpha=1)
   n40 = Branch(
    execute=@n27.then_0
    Condition=OR Boolean!(A=Get bIsSkydiveFloating(), B=Get bIsSkydiveDiveMode(), C:bool=Get bIsParachuteOpened(), D:bool=Get bIsSkyDiving()) -> (ReturnValue)
   )
   > Set SkirtDynAlpha(SkirtDynAlpha:double=0.25) -> (Output_Get:double)
   Set SkirtDynAlpha(execute=@n40.else, SkirtDynAlpha=1)
   > n41 = Branch(Condition=AND Boolean(A=Get isWearingCape() -> (isWearingCape:bool), B=Get bIsCrouching()))
   > Set SkirtDynAlpha(SkirtDynAlpha=0.5)
   Set SkirtDynAlpha(execute=@n41.else, SkirtDynAlpha=1)
   n42 = Sequence(execute=@n1.then_3)
   n43 = Branch(execute=@n42.then_1, Condition=Get bIsInFrontEnd())
   > Set FrontEndAlpha(FrontEndAlpha:double=0) -> (Output_Get:double)
   Set FrontEndAlpha(execute=@n43.else, FrontEndAlpha=1)
   n44 = Get bIsPlayingEmote()
   n45 = Branch(execute=@n42.then_0, Condition=AND Boolean(A=Get bIsInFrontEnd(), B=NOT Boolean(A=@n44)))
   > Set SkirtDynAlpha_FE(SkirtDynAlpha_FE:double=0.5) -> (Output_Get:double)
   Set SkirtDynAlpha_FE(execute=@n45.else, SkirtDynAlpha_FE=1)
   n46 = Branch(execute=@n42.then_2, Condition=AND Boolean(A=Get bIsInFrontEnd(), B=Get bIsPlayingEmote()))
   > Set SleevesDynAlpha_FE(SleevesDynAlpha_FE:double=0.5) -> (Output_Get:double)
   Set SleevesDynAlpha_FE(execute=@n46.else, SleevesDynAlpha_FE=1)
   n47 = Sequence(execute=@n1.then_0)
   n48 = BreakTransform(InTransform=GetDeltaTransformFromRefPose(self=GetOwnerSkeletalMeshComponent(), BoneName=upperarm_r, BaseName=clavicle_r))
   n49 = Branch(execute=@n47.then_1, Condition="float >= float"(A=BreakRotator(InRot=@n48.Rotation).Yaw, B=90))
   > Set UpperArmRot_l(UpperArmRot_l:bool=1) -> (Output_Get:bool)
   Set UpperArmRot_l(execute=@n49.else, UpperArmRot_l=false)
   n50 = Gravity Override State Function DEPRECATED(
    execute=@n47.then_2
    CharacterPartAnimInstance=Self-Reference()
    GameTimeInSeconds=GetGameTimeInSeconds()
    GravityOverrideParams=Get Gravity Override Params()
   )
   > Set GravityOverride(GravityOverride=@n50)
   n51 = Cast To FortPlayerAnimInstance(execute=@n47.then_0, Object=GetAnimInstance(self=GetOwnerSkeletalMeshComponent()))
   > Set ar_firingPosition_GC(ar_firingPosition_GC:int=0) -> (Output_Get:int=0)
   > n52 = Switch on EFortWeaponCoreAnimation!(execute, Selection:EFortWeaponCoreAnimation Enum=Get WeaponCoreAnim(self=@n51)) -> (
    Melee:Exec
    Pistol:Exec
    Shotgun:Exec
    PaperBlueprint:Exec
    Rifle:Exec
    MeleeOneHand:Exec
    MachinePistol:Exec
    RocketLauncher:Exec
    GrenadeLauncher:Exec
    GoingCommando:Exec
    AssaultRifle:Exec
    TacticalShotgun:Exec
    SniperRifle:Exec
    TrapPlacement:Exec
    ShoulderLauncher:Exec
    AbilityDecoTool:Exec
    Crossbow:Exec
    C4:Exec
    RemoteControl:Exec
    DualWield:Exec
    AR_BullPup:Exec
    AR_ForwardGrip:Exec
    MedPackPaddles:Exec
    SMG_P90:Exec
    AR_DrumGun:Exec
    Consumable_Small:Exec
    Consumable_Large:Exec
    Balloon:Exec
    MountedTurret:Exec
    CreativeTool:Exec
    ExplosiveBow:Exec
    AshtonIndigo:Exec
    AshtonChicago:Exec
    MeleeDualWield:Exec
    Unarmed:Exec
   )
   n53 = Branch(execute=[
     @n52.AssaultRifle
     @n52.GoingCommando
     @n52.MachinePistol
     @n52.Pistol
     @n52.Rifle
     @n52.RocketLauncher
     @n52.Shotgun
     @n52.SniperRifle
     @n52.TacticalShotgun
    ])
   > n54 = Set ar_firingPosition_GC(ar_firingPosition_GC=1) -> (Output_Get=0)
   > Set Gun_out(Gun_out:bool="To Boolean (Integer)"!(InInt:int=@n54) -> (ReturnValue:bool)) -> (Output_Get:bool)
   Set ar_firingPosition_GC(execute=@n53.else, ar_firingPosition_GC=0) -> (Output_Get=0)
   Event OnCharacterCustomizationLoaded()
   > Set BodyMesh(BodyMesh=GetOwnerSkeletalMeshComponent())
   > n55 = Cast To PlayerPawn_Athena(Object=TryGetPawnOwner())
   > Set PawnOwner(PawnOwner:Player Pawn Athena Object Reference=@n55) -> (Output_Get:Player Pawn Athena Object Reference)
   n56 = CheckForBackbling()
   > Set isWearingCape(isWearingCape=OR Boolean!(A=@n56.bIsBackblingACape, B=@n56.bIsBackblingLong) -> (ReturnValue))
   TryGetPawnOwner()
  ]}
]
variables [
 ar_firingPosition_GC {type=integer}
 BigGunOut {type=boolean}
 BodyMesh {type=SkeletalMeshComponent}
 CapeFix {type=number}
 CommandAlpha {type=boolean}
 Crouch {type=number}
 CrouchGun {type=number}
 crouchMove {type=number}
 CrouchRBdisabled {type=boolean}
 crouchSleeve {type=number}
 DBNOMode {type=number}
 Emoting {type=number}
 falling {type=number}
 FE_Idle {type=number}
 FrontEndAlpha {type=number}
 Gravity Override Params {type=GravityOverrideParamsStruct}
 GravityOverride {type=Vector}
 Gun_out {type=boolean}
 isWearingCape {type=boolean}
 jumping {type=number}
 moving2D {type=number}
 NewVar {type=boolean}
 ParachuteOpened {type=number}
 PawnOwner {type=PlayerPawn_Athena_C}
 Running {type=number}
 SkirtDynAlpha {type=number}
 SkirtDynAlpha_FE {type=number}
 Skydiving {type=number}
 SleevesDynAlpha {type=number}
 SleevesDynAlpha_FE {type=number}
 Sliding {type=number}
 SlidingRBdisabled {type=boolean}
 Sprinting Crouched {type=number}
 SprintRun {type=number}
 TacticalJump {type=number}
 targetUpAlpha {type=number}
 tight_l_rot {type=number}
 UpperArmRot_l {type=boolean}
]
