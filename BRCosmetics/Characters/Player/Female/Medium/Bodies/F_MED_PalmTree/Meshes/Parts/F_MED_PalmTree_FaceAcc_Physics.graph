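# Physics asset for the face accessory mesh of the female medium PalmTree character, built around its hair bones.
# Capsule bodies cover upper spine, clavicles, neck and head, linked by free-motion constraints, with five base hair bodies attached freely to the head.
# Dynamic hair bodies branch from those bases as left and right side chains and three back masters fanning into strands.
# Hair joints use tight swing limits and zero-degree twist limits, restricting strand movement.
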PhysicsAsset /BRCosmetics/Characters/Player/Female/Medium/Bodies/F_MED_PalmTree/Meshes/Parts/F_MED_PalmTree_FaceAcc_Physics
bodies {
 base_hairDn_A_1_r [{shapeType=Capsule, length=0.505, radius=0.505, shapeName=base_hairDn_A_1_r_capsule, rotation=<rot 0 -90 0>}]
 base_hair_A_1_l [{shapeType=Capsule, length=0.505, radius=0.505, shapeName=base_hair_A_1_l_capsule, rotation=<rot 0 -90 0>}]
 base_master_hairDn_back [{shapeType=Capsule, length=0.505, radius=0.505, shapeName=base_master_hairDn_back_capsule, rotation=<rot 0 -90 0>}]
 base_master_hairDn_l_back [{shapeType=Capsule, length=0.505, radius=0.505, shapeName=base_master_hairDn_l_back_capsule, rotation=<rot 0 -90 0>}]
 base_master_hairDn_r_back [{shapeType=Capsule, length=0.505, radius=0.505, shapeName=base_master_hairDn_r_back_capsule, rotation=<rot 0 -90 0>}]
 clavicle_l [{shapeType=Capsule, length=10, radius=4.6, shapeName=clavicle_l_capsule, center=<3 0 0>, rotation=<rot 90 0 0>}]
 clavicle_r [{shapeType=Capsule, length=10, radius=4.6, shapeName=clavicle_r_capsule, center=<-3 0 0>, rotation=<rot 90 0 0>}]
 dyn_hairDn_A_01_back [{shapeType=Capsule, length=1, radius=0.6875, shapeName=dyn_hairDn_A_01_back_capsule, center=<4.9819 0 0>, rotation=<rot 90 0 0>}]
 dyn_hairDn_A_1_r [{shapeType=Capsule, length=0.5, radius=0.8, shapeName=dyn_hairDn_A_1_r_capsule, center=<1.0553 0 0>, rotation=<rot 90 0 0>}]
 dyn_hairDn_A_2_r [{shapeType=Capsule, length=0.5, radius=0.8, shapeName=dyn_hairDn_A_2_r_capsule, center=<1.0553 0 0>, rotation=<rot 90 0 0>}]
 dyn_hairDn_A_3_r [{shapeType=Capsule, length=1, radius=1, shapeName=dyn_hairDn_A_3_r_capsule, center=<1.8655 0 0>, rotation=<rot 90 0 0>}]
 dyn_hairDn_B_01_back [{shapeType=Capsule, length=1, radius=0.7, shapeName=dyn_hairDn_B_01_back_capsule, center=<5.1013 0 0>, rotation=<rot 90 0 0>}]
 dyn_hairDn_C_01_back [{shapeType=Capsule, length=1, radius=0.6875, shapeName=dyn_hairDn_C_01_back_capsule, center=<5.8422 0 0>, rotation=<rot 90 0 0>}]
 dyn_hairDn_D_01_back [{shapeType=Capsule, length=1, radius=0.6875, shapeName=dyn_hairDn_D_01_back_capsule, center=<5.6094 0 0>, rotation=<rot 90 0 0>}]
 dyn_hairDn_E_01_back [{shapeType=Capsule, length=1, radius=1.3, shapeName=dyn_hairDn_E_01_back_capsule, center=<6.31 0.7508 -0.6299>, rotation=<rot 90 0 0>}]
 dyn_hairDn_H_01_back [{shapeType=Capsule, length=1, radius=1, shapeName=dyn_hairDn_H_01_back_capsule, center=<5.9869 0 0.5384>, rotation=<rot 84.375 180 179.9999>}]
 dyn_hairDn_I_01_back [{shapeType=Capsule, length=1, radius=1.3, shapeName=dyn_hairDn_I_01_back_capsule, center=<6.2451 0 0>, rotation=<rot 90 0 0>}]
 dyn_hairDn_J_01_back [{shapeType=Capsule, length=1, radius=1, shapeName=dyn_hairDn_J_01_back_capsule, center=<4.5259 0 0>, rotation=<rot 90 0 0>}]
 dyn_hairDn_L_01_back [{shapeType=Capsule, length=1, radius=1, shapeName=dyn_hairDn_L_01_back_capsule, center=<5.1232 0 0>, rotation=<rot 90 0 0>}]
 dyn_hairDn_N_01_back [{shapeType=Capsule, length=1, radius=1, shapeName=dyn_hairDn_N_01_back_capsule, center=<5.13 0 0>, rotation=<rot 90 0 0>}]
 dyn_hair_A_1_l [{shapeType=Capsule, length=0.5, radius=1, shapeName=dyn_hair_A_1_l_capsule, center=<1.0553 0 0>, rotation=<rot 90 0 0>}]
 dyn_hair_A_2_l [{shapeType=Capsule, length=0.5, radius=1, shapeName=dyn_hair_A_2_l_capsule, center=<1.0553 0 0>, rotation=<rot 90 0 0>}]
 dyn_hair_A_3_l [{shapeType=Capsule, length=0.5, radius=1, shapeName=dyn_hair_A_3_l_capsule, center=<1.0553 0 0>, rotation=<rot 90 0 0>}]
 dyn_hair_A_5_l [{shapeType=Capsule, length=1, radius=1, shapeName=dyn_hair_A_5_l_capsule, center=<1 0 0>, rotation=<rot 90 0 0>}]
 dyn_master_hairDn_back [{shapeType=Capsule, length=0.8125, radius=0.75, shapeName=dyn_master_hairDn_back_capsule, center=<0.7353 -0.7631 0>, rotation=<rot 90 0 0>}]
 dyn_master_hairDn_l_back [{shapeType=Capsule, length=0.75, radius=0.8125, shapeName=dyn_master_hairDn_l_back_capsule, center=<0.7353 0 0>, rotation=<rot 90 0 0>}]
 dyn_master_hairDn_r_back [{shapeType=Capsule, length=0.75, radius=1.0625, shapeName=dyn_master_hairDn_r_back_capsule, center=<0.7353 0 0>, rotation=<rot 90 0 0>}]
 head [{shapeType=Capsule, length=0.5, radius=0.5}]
 neck_02 [{shapeType=Capsule, length=5, radius=5, shapeName=neck_02_capsule, rotation=<rot 90 0 0>}]
 spine_05 [{shapeType=Capsule, length=17.375, radius=8, shapeName=spine_05_capsule, center=<3.8838 -2.8761 0>, rotation=<rot 0 -81.5626 0>}]
}
constraints [
 {
  bone1Name=base_hairDn_A_1_r
  bone2Name=head
  swing1LimitDegrees=45
  swing1Motion=Free
  swing2LimitDegrees=45
  swing2Motion=Free
  twistLimitDegrees=45
  twistMotion=Free
 }
 {
  bone1Name=base_hair_A_1_l
  bone2Name=head
  swing1LimitDegrees=45
  swing1Motion=Free
  swing2LimitDegrees=45
  swing2Motion=Free
  twistLimitDegrees=45
  twistMotion=Free
 }
 {bone1Name=dyn_hairDn_A_1_r, bone2Name=base_hairDn_A_1_r, swing1LimitDegrees=15, swing2LimitDegrees=10, twistLimitDegrees=0}
 {bone1Name=dyn_hairDn_A_2_r, bone2Name=dyn_hairDn_A_1_r, swing1LimitDegrees=10, swing2LimitDegrees=5, twistLimitDegrees=0}
 {bone1Name=dyn_hairDn_A_3_r, bone2Name=dyn_hairDn_A_1_r, swing1LimitDegrees=10, swing2LimitDegrees=5, twistLimitDegrees=0}
 {bone1Name=dyn_hair_A_1_l, bone2Name=base_hair_A_1_l, swing1LimitDegrees=2, swing2LimitDegrees=2, twistLimitDegrees=0}
 {bone1Name=dyn_hair_A_2_l, bone2Name=dyn_hair_A_1_l, swing1LimitDegrees=2, swing2LimitDegrees=2, twistLimitDegrees=0}
 {bone1Name=dyn_hair_A_3_l, bone2Name=dyn_hair_A_2_l, swing1LimitDegrees=10, swing2LimitDegrees=5, twistLimitDegrees=0}
 {
  bone1Name=base_master_hairDn_back
  bone2Name=head
  swing1LimitDegrees=45
  swing1Motion=Free
  swing2LimitDegrees=45
  swing2Motion=Free
  twistLimitDegrees=45
  twistMotion=Free
 }
 {
  bone1Name=base_master_hairDn_l_back
  bone2Name=head
  swing1LimitDegrees=45
  swing1Motion=Free
  swing2LimitDegrees=45
  swing2Motion=Free
  twistLimitDegrees=45
  twistMotion=Free
 }
 {
  bone1Name=base_master_hairDn_r_back
  bone2Name=head
  swing1LimitDegrees=45
  swing1Motion=Free
  swing2LimitDegrees=45
  swing2Motion=Free
  twistLimitDegrees=45
  twistMotion=Free
 }
 {bone1Name=dyn_master_hairDn_back, bone2Name=base_master_hairDn_back, swing1LimitDegrees=6, swing2LimitDegrees=5, twistLimitDegrees=0}
 {bone1Name=dyn_hairDn_E_01_back, bone2Name=dyn_master_hairDn_back, swing1LimitDegrees=6, swing2LimitDegrees=5, twistLimitDegrees=0}
 {bone1Name=dyn_hairDn_H_01_back, bone2Name=dyn_master_hairDn_back, swing1LimitDegrees=6, swing2LimitDegrees=5, twistLimitDegrees=0}
 {bone1Name=dyn_hairDn_I_01_back, bone2Name=dyn_master_hairDn_back, swing1LimitDegrees=6, swing2LimitDegrees=5, twistLimitDegrees=0}
 {bone1Name=dyn_master_hairDn_l_back, bone2Name=base_master_hairDn_l_back, swing1LimitDegrees=10, swing2LimitDegrees=5, twistLimitDegrees=0}
 {bone1Name=dyn_hairDn_A_01_back, bone2Name=dyn_master_hairDn_l_back, swing1LimitDegrees=10, swing2LimitDegrees=5, twistLimitDegrees=0}
 {bone1Name=dyn_hairDn_B_01_back, bone2Name=dyn_master_hairDn_l_back, swing1LimitDegrees=20, swing2LimitDegrees=20, twistLimitDegrees=0}
 {bone1Name=dyn_hairDn_C_01_back, bone2Name=dyn_master_hairDn_l_back, swing1LimitDegrees=10, swing2LimitDegrees=5, twistLimitDegrees=0}
 {bone1Name=dyn_hairDn_D_01_back, bone2Name=dyn_master_hairDn_l_back, swing1LimitDegrees=10, swing2LimitDegrees=5, twistLimitDegrees=0}
 {bone1Name=dyn_master_hairDn_r_back, bone2Name=base_master_hairDn_r_back, swing1LimitDegrees=10, swing2LimitDegrees=5, twistLimitDegrees=0}
 {
  bone1Name=dyn_hairDn_J_01_back
  bone2Name=dyn_master_hairDn_r_back
  swing1LimitDegrees=25
  swing2LimitDegrees=10
  twistLimitDegrees=0
  twistMotion=Locked
 }
 {bone1Name=dyn_hairDn_L_01_back, bone2Name=dyn_master_hairDn_r_back, swing1LimitDegrees=10, swing2LimitDegrees=5, twistLimitDegrees=0}
 {bone1Name=dyn_hairDn_N_01_back, bone2Name=dyn_master_hairDn_r_back, swing1LimitDegrees=10, swing2LimitDegrees=5, twistLimitDegrees=0}
 {
  bone1Name=head
  bone2Name=neck_02
  swing1LimitDegrees=45
  swing1Motion=Free
  swing2LimitDegrees=45
  swing2Motion=Free
  twistLimitDegrees=45
  twistMotion=Free
 }
 {
  bone1Name=neck_02
  bone2Name=spine_05
  swing1LimitDegrees=45
  swing1Motion=Free
  swing2LimitDegrees=45
  swing2Motion=Free
  twistLimitDegrees=45
  twistMotion=Free
 }
 {bone1Name=dyn_hair_A_5_l, bone2Name=dyn_hair_A_2_l, swing1LimitDegrees=1, swing2LimitDegrees=6, twistLimitDegrees=0}
 {
  bone1Name=clavicle_l
  bone2Name=spine_05
  swing1LimitDegrees=45
  swing1Motion=Free
  swing2LimitDegrees=45
  swing2Motion=Free
  twistLimitDegrees=45
  twistMotion=Free
 }
 {
  bone1Name=clavicle_r
  bone2Name=spine_05
  swing1LimitDegrees=45
  swing1Motion=Free
  swing2LimitDegrees=45
  swing2Motion=Free
  twistLimitDegrees=45
  twistMotion=Free
 }
]
